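# Animation Level Sequence for the mmh1000 cinematic, staging the MaskMaker NPC, the TentacleBoss and a rift portal despawn Niagara effect.
# MaskMaker toggles visible, grows in stepped scale keys and shifts sharply in position near the end, over three skeletal animation sections.
# TentacleBoss enters later with four overlapping animation sections and a vertical dip and rise, while its four body material slots hold a constant WorldRootGlowIntensity.
# The portal actor only has a fixed transform.
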LevelSequence /PistolCanary/Sequence/mmh/mmh1000/PistolCanary_mmh1000_anm
bindings [
 PistolCanary_CinematicNPC_MaskMaker {
  id=#1
  possessed_class=PistolCanary_CinematicNPC_MaskMaker_C
  tracks [
   {
    type=MovieScene3DTransformTrack
    sections [{
      type=MovieScene3DTransformSection
      channels [
       Location.X {type=MovieSceneScriptingDoubleChannel, default=-100, const [-100,3]}
       Location.Y {type=MovieSceneScriptingDoubleChannel, default=9500, keys [[3763,-6320.5469],[8014,-6320.5469],[8136,9500]]}
       Location.Z {type=MovieSceneScriptingDoubleChannel, default=8000, keys [[3729,1100],[8014,1100],[8130,8000]]}
       Rotation.X {type=MovieSceneScriptingDoubleChannel, default=-52, const [0,3]}
       Rotation.Y {type=MovieSceneScriptingDoubleChannel, default=5, const [0,3]}
       Rotation.Z {type=MovieSceneScriptingDoubleChannel, default=256, const [0,3]}
       Scale.X {
        type=MovieSceneScriptingDoubleChannel
        default=2.45
        thinned_from=17
        keys [[3729,1.25],[5044,1.25],[5105,1.75],[5166,1.75],[5198,1.95],[5211,2],[5275,2],[5284,2.1],[5298,2.35],[5304,2.45],[5312,2.5],[5391,2.5],[5429,3],[8133,3]]
       }
       Scale.Y {
        type=MovieSceneScriptingDoubleChannel
        default=2.45
        thinned_from=17
        keys [[3729,1.25],[5044,1.25],[5105,1.75],[5166,1.75],[5198,1.95],[5211,2],[5275,2],[5284,2.1],[5298,2.35],[5304,2.45],[5312,2.5],[5391,2.5],[5429,3],[8133,3]]
       }
       Scale.Z {
        type=MovieSceneScriptingDoubleChannel
        default=2.45
        thinned_from=17
        keys [[3729,1.25],[5044,1.25],[5105,1.75],[5166,1.75],[5198,1.95],[5211,2],[5275,2],[5284,2.1],[5298,2.35],[5304,2.45],[5312,2.5],[5391,2.5],[5429,3],[8133,3]]
       }
      ]
     }]
   }
   {
    type=MovieSceneVisibilityTrack
    sections [{
      type=MovieSceneVisibilitySection
      channels [None {type=MovieSceneScriptingBoolChannel, default=false, keys [[3600,false],[3675,true],[8137,false]]}]
     }]
   }
   {
    type=MovieSceneSkeletalAnimationTrack
    sections [
     {type=MovieSceneSkeletalAnimationSection, end_frame=7029, start_frame=3600}
     {type=MovieSceneSkeletalAnimationSection, end_frame=8014, start_frame=6932}
     {type=MovieSceneSkeletalAnimationSection, end_frame=8267, start_frame=8014}
    ]
   }
  ]
 }
 NS_SourFork_RiftGate_Portal_Despawn {
  id=#2
  possessed_class=NiagaraActor
  tracks [{
    type=MovieScene3DTransformTrack
    sections [{
      type=MovieScene3DTransformSection
      channels [
       Location.X {type=MovieSceneScriptingDoubleChannel, default=14328, const [14328,1]}
       Location.Y {type=MovieSceneScriptingDoubleChannel, default=70128, const [70128,1]}
       Location.Z {type=MovieSceneScriptingDoubleChannel, default=3896, const [3896,1]}
       Rotation.X {type=MovieSceneScriptingDoubleChannel, default=0, const [0,1]}
       Rotation.Y {type=MovieSceneScriptingDoubleChannel, default=0, const [0,1]}
       Rotation.Z {type=MovieSceneScriptingDoubleChannel, default=0, const [0,1]}
       Scale.X {type=MovieSceneScriptingDoubleChannel, default=1.2, const [1.2,1]}
       Scale.Y {type=MovieSceneScriptingDoubleChannel, default=1.2, const [1.2,1]}
       Scale.Z {type=MovieSceneScriptingDoubleChannel, default=1, const [1,1]}
      ]
     }]
   }]
 }
 TentacleBoss {
  id=#3
  possessed_class=PistolCanary_TentacleBoss_C
  tracks [
   {
    type=MovieSceneSkeletalAnimationTrack
    sections [
     {type=MovieSceneSkeletalAnimationSection, end_frame=7287, start_frame=6986}
     {type=MovieSceneSkeletalAnimationSection, end_frame=7777, start_frame=7157}
     {type=MovieSceneSkeletalAnimationSection, end_frame=8092, start_frame=7765}
     {type=MovieSceneSkeletalAnimationSection, end_frame=8717, start_frame=8038}
    ]
   }
   {
    type=MovieScene3DTransformTrack
    sections [{
      type=MovieScene3DTransformSection
      channels [
       Location.X {type=MovieSceneScriptingDoubleChannel, default=0, const [123.5116,1]}
       Location.Y {type=MovieSceneScriptingDoubleChannel, default=-5000, const [-6000,1]}
       Location.Z {type=MovieSceneScriptingDoubleChannel, default=500, keys [[6986,0],[7052,-3000],[7223,0],[7651,500]]}
       Rotation.X {type=MovieSceneScriptingDoubleChannel, default=0, const [0,1]}
       Rotation.Y {type=MovieSceneScriptingDoubleChannel, default=-0.1356, const [-0.1356,1]}
       Rotation.Z {type=MovieSceneScriptingDoubleChannel, default=0, const [0,1]}
       Scale.X {type=MovieSceneScriptingDoubleChannel, default=0.75, const [0.75,1]}
       Scale.Y {type=MovieSceneScriptingDoubleChannel, default=0.75, const [0.75,1]}
       Scale.Z {type=MovieSceneScriptingDoubleChannel, default=0.75, const [0.75,1]}
      ]
     }]
   }
  ]
 }
 SkeletalMesh {
  id=#4
  parent_id=#3
  possessed_class=SkeletalMeshComponent
  tracks [
   {
    type=MovieSceneComponentMaterialTrack
    display_name=Material Slot: Body_A
    sections [{
      type=MovieSceneComponentMaterialParameterSection
      end_frame=8886
      start_frame=3600
      channels [WorldRootGlowIntensity {type=MovieSceneScriptingFloatChannel, default=0.1, const [0.1,1]}]
     }]
   }
   {
    type=MovieSceneComponentMaterialTrack
    display_name=Material Slot: Body_B
    sections [{
      type=MovieSceneComponentMaterialParameterSection
      end_frame=8886
      start_frame=3600
      channels [WorldRootGlowIntensity {type=MovieSceneScriptingFloatChannel, default=0.1, const [0.1,1]}]
     }]
   }
   {
    type=MovieSceneComponentMaterialTrack
    display_name=Material Slot: Body_C
    sections [{
      type=MovieSceneComponentMaterialParameterSection
      end_frame=8886
      start_frame=3600
      channels [WorldRootGlowIntensity {type=MovieSceneScriptingFloatChannel, default=0.1, const [0.1,1]}]
     }]
   }
   {
    type=MovieSceneComponentMaterialTrack
    display_name=Material Slot: Body_D
    sections [{
      type=MovieSceneComponentMaterialParameterSection
      end_frame=8886
      start_frame=3600
      channels [WorldRootGlowIntensity {type=MovieSceneScriptingFloatChannel, default=0.1, const [0.1,1]}]
     }]
   }
  ]
 }
]
display_rate {denominator=1, numerator=60}
tick_resolution {denominator=1, numerator=24000}
tracks []
