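# PistolCanary level sequence for shot mmh1000 (anm1) that animates a cinematic with the MaskMaker NPC, a tentacle boss and a looping Niagara aura.
# Transform keys move and rotate the MaskMaker while it grows in steps, with a visibility toggle at the end.
# The boss enters later, rising from far below on keyed transforms over six skeletal animation sections, the later ones overlapping.
# The aura holds a static transform; there are no camera or audio tracks.
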LevelSequence /PistolCanary/Sequence/mmh/mmh1000/PistolCanary_mmh1000_anm1
bindings [
 NS Pistol Canary MM Aura Loop {
  id=#1
  possessed_class=NiagaraActor
  tracks [{
    type=MovieScene3DTransformTrack
    sections [{
      type=MovieScene3DTransformSection
      channels [
       Location.X {type=MovieSceneScriptingDoubleChannel, default=4195.179, const [-43.2698,1]}
       Location.Y {type=MovieSceneScriptingDoubleChannel, default=2892.5476, const [-23.3521,1]}
       Location.Z {type=MovieSceneScriptingDoubleChannel, default=-2491.5478, const [29.7573,1]}
       Rotation.X {type=MovieSceneScriptingDoubleChannel, default=0, const [0,1]}
       Rotation.Y {type=MovieSceneScriptingDoubleChannel, default=0, const [0,1]}
       Rotation.Z {type=MovieSceneScriptingDoubleChannel, default=0, const [0,1]}
       Scale.X {type=MovieSceneScriptingDoubleChannel, default=1, const [1,1]}
       Scale.Y {type=MovieSceneScriptingDoubleChannel, default=1, const [1,1]}
       Scale.Z {type=MovieSceneScriptingDoubleChannel, default=1, const [1,1]}
      ]
     }]
   }]
 }
 TentacleBoss {
  id=#2
  possessed_class=PistolCanary_TentacleBoss_C
  tracks [
   {
    type=MovieSceneSkeletalAnimationTrack
    sections [
     {type=MovieSceneSkeletalAnimationSection, end_frame=5883, start_frame=5587}
     {type=MovieSceneSkeletalAnimationSection, end_frame=6248, start_frame=5883}
     {type=MovieSceneSkeletalAnimationSection, end_frame=6567, start_frame=6248}
     {type=MovieSceneSkeletalAnimationSection, end_frame=6861, start_frame=6464}
     {type=MovieSceneSkeletalAnimationSection, end_frame=7044, start_frame=6784}
     {type=MovieSceneSkeletalAnimationSection, end_frame=7608, start_frame=6939}
    ]
   }
   {
    type=MovieScene3DTransformTrack
    sections [{
      type=MovieScene3DTransformSection
      channels [
       Location.X {type=MovieSceneScriptingDoubleChannel, default=375, keys [[5804.0075,679.9999],[6318,679.9999],[6725,375],[6863,-517],[6939,290]]}
       Location.Y {
        type=MovieSceneScriptingDoubleChannel
        default=-6454
        keys [[5804.0075,-6937.6665],[6318,-4808],[6478,-6317],[6725,-6696],[6803,-6454],[6883,-8920],[6939,-6544]]
       }
       Location.Z {
        type=MovieSceneScriptingDoubleChannel
        default=-5268
        keys [[5804.0075,-16000],[6031,-12585.1582],[6054,-12514.2812],[6117,-12255.4014],[6248,-7302],[6505,-4267],[6579,-4788],[6652,-4267],[6725,-2196.4924],[6758,-1675],[6823,-1675],[6863,-6667],[6883,-5268],[6939,-1600]]
       }
       Rotation.X {
        type=MovieSceneScriptingDoubleChannel
        default=-30
        thinned_from=9
        keys [[5804.0075,0],[6464.01,0],[6648,5],[6701.99,0],[6784,0],[6863,-30],[6887.99,-10],[6939,0]]
       }
       Rotation.Y {
        type=MovieSceneScriptingDoubleChannel
        default=-2
        keys [[5804.0075,0],[6396,0],[6579,-5],[6648,-2],[6701.99,0],[6784,0],[6863,-22],[6939,0]]
       }
       Rotation.Z {type=MovieSceneScriptingDoubleChannel, default=0, thinned_from=3, keys [[5804.0075,0],[6648,0]]}
       Scale.X {type=MovieSceneScriptingDoubleChannel, default=1, const [1,2]}
       Scale.Y {type=MovieSceneScriptingDoubleChannel, default=1, const [1,2]}
       Scale.Z {type=MovieSceneScriptingDoubleChannel, default=1, const [1,2]}
      ]
     }]
   }
  ]
 }
 PistolCanary_CinematicNPC_MaskMaker {
  id=#3
  possessed_class=PistolCanary_CinematicNPC_MaskMaker_C
  tracks [
   {
    type=MovieScene3DTransformTrack
    sections [{
      type=MovieScene3DTransformSection
      channels [
       Location.X {
        type=MovieSceneScriptingDoubleChannel
        default=-46555
        thinned_from=29
        keys [[3729,-123.8243],[3833,-107.9144],[4178,-83.9955],[4279,-83.8243],[4323,-84.6213],[4514,-277],[4643,-277],[4652,-293],[4674,-277],[4873,-277],[4881,-252],[4904,-277],[5300.005,-277],[5382.0025,-278.5337],[5409.0025,-277],[5998.0075,-294.6389],[6056.0075,-278.0432],[6505,-294.9147],[6528,-367.1645],[6572,-377.5912],[6863,-294.9147],[6958,-46555]]
       }
       Location.Y {
        type=MovieSceneScriptingDoubleChannel
        default=8164
        thinned_from=29
        keys [[3729,-5876.5469],[3833,-3553.2683],[4178,-60.4406],[4279,-35.4355],[4323,-44.9873],[4514,-2061],[4769,-2061],[4774,-2121],[4790,-2061],[4904,-2061],[4964,-2051.9248],[4974,-1966],[4992,-2043.6404],[5196.0025,-1907.1669],[5229.005,-1810.689],[5264.0025,-1872.3558],[5300.005,-1373.3442],[5382.0025,-1799.0197],[5409.0025,-1768.4922],[5998.0075,-2012.5377],[6056.0075,-2058.1338],[6505,-2011.7798],[6528,-1813.275],[6572,-1784.6281],[6863,-2011.7798],[6958,8164]]
       }
       Location.Z {
        type=MovieSceneScriptingDoubleChannel
        default=1585
        thinned_from=31
        keys [[3729,791.7608],[3897,1004],[4178,1585],[4279,1526],[4308,1494],[4514,5000],[4567,5052],[4643,5021.9658],[4735,4955],[4769,4975.4795],[4849,5081],[4873,5052.3438],[4904,5000],[5063,4886.0254],[5196.0025,5028],[5264,5028],[5382.0025,4859.479],[5409.0025,4891.2686],[6056,4891.2686],[6505,5646.1782],[6528,5455.3071],[6572,5299.5884],[6863,5646.1782],[6904,13676],[6931,15257],[6958,13706]]
       }
       Rotation.X {
        type=MovieSceneScriptingDoubleChannel
        default=55
        thinned_from=23
        keys [[3729,0],[4567,0],[4674,-20],[4790,-10],[4904,-20],[5008.0025,-4.3704],[5196.0025,20],[5300.005,41.9809],[5331.005,43.7128],[5813,26.0438],[6022,46],[6117,55],[6361,19],[6464.01,19],[6528,-18.0554],[6538,-20],[6863,-20],[6887.99,120],[6958,120]]
       }
       Rotation.Y {type=MovieSceneScriptingDoubleChannel, default=0, thinned_from=22, keys [[3729,0],[6958,0]]}
       Rotation.Z {
        type=MovieSceneScriptingDoubleChannel
        default=0
        thinned_from=22
        keys [[3729,0],[4308,0],[4404,165.9097],[4514,180.0823],[5331.005,180],[5998.0075,200],[6863,200],[6887.99,41],[6958,41]]
       }
       Scale.X {
        type=MovieSceneScriptingDoubleChannel
        default=2
        thinned_from=15
        keys [[3729,1.25],[4653,1.25],[4712,1.75],[4772,1.75],[4824,2],[4875,2],[4919,2.5],[4951.0025,2.5],[5008.0025,3],[6958,3]]
       }
       Scale.Y {
        type=MovieSceneScriptingDoubleChannel
        default=2
        thinned_from=15
        keys [[3729,1.25],[4653,1.25],[4712,1.75],[4772,1.75],[4824,2],[4875,2],[4919,2.5],[4951.0025,2.5],[5008.0025,3],[6958,3]]
       }
       Scale.Z {
        type=MovieSceneScriptingDoubleChannel
        default=2
        thinned_from=15
        keys [[3729,1.25],[4653,1.25],[4712,1.75],[4772,1.75],[4824,2],[4875,2],[4919,2.5],[4951.0025,2.5],[5008.0025,3],[6958,3]]
       }
      ]
     }]
   }
   {
    type=MovieSceneVisibilityTrack
    sections [{type=MovieSceneVisibilitySection, channels [None {type=MovieSceneScriptingBoolChannel, default=false, keys [[6969,true],[6970,false]]}]}]
   }
  ]
 }
]
display_rate {denominator=1, numerator=60}
tick_resolution {denominator=1, numerator=24000}
tracks []
