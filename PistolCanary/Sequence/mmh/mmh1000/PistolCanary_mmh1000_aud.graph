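# Audio sub-sequence for the PistolCanary mmh1000 cinematic, scoring the MaskMaker, portal, monolith spires, tentacle boss and house.
# Master tracks carry music, MaskMaker sounds and dialogue, tentacle and house audio, with keyed volume ducking.
# Bound spire and portal actors play their own loop and portal sections.
# The AudioManager receives timed events, an event-mix toggle and a stepped Music Phase integer, while MaskMaker and TentacleBoss bindings fire events.
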LevelSequence /PistolCanary/Sequence/mmh/mmh1000/PistolCanary_mmh1000_aud
bindings [
 BP_PistolCanary_LightBeam {id=#1, possessed_class=BP_PistolCanary_LightBeam_C, tracks []}
 PistolCanary_RockSpireMonolith4 {
  id=#2
  possessed_class=PistolCanary_Monolith_Spire_C
  tracks [{
    type=MovieSceneAudioTrack
    display_name=Loop
    sections [
     {
      type=MovieSceneAudioSection
      end_frame=5961
      start_frame=5418
      channels [
       Volume {type=MovieSceneScriptingFloatChannel, default=1}
       Pitch {type=MovieSceneScriptingFloatChannel, default=1}
      ]
     }
     {
      type=MovieSceneAudioSection
      end_frame=6355
      start_frame=5418
      channels [
       Volume {type=MovieSceneScriptingFloatChannel, default=1}
       Pitch {type=MovieSceneScriptingFloatChannel, default=1}
      ]
     }
    ]
   }]
 }
 BP_MaskMaker_Portal_PistolCanary {
  id=#3
  possessed_class=BP_MaskMaker_Portal_PistolCanary_C
  tracks [{
    type=MovieSceneAudioTrack
    display_name=PortalAudio
    sections [
     {
      type=MovieSceneAudioSection
      end_frame=6561
      start_frame=6162
      channels [
       Volume {type=MovieSceneScriptingFloatChannel, default=1}
       Pitch {type=MovieSceneScriptingFloatChannel, default=1}
      ]
     }
     {
      type=MovieSceneAudioSection
      end_frame=4316
      start_frame=3723
      channels [
       Volume {type=MovieSceneScriptingFloatChannel, default=1}
       Pitch {type=MovieSceneScriptingFloatChannel, default=1}
      ]
     }
    ]
   }]
 }
 PistolCanary_RockSpireMonolith3 {
  id=#4
  possessed_class=PistolCanary_Monolith_Spire_C
  tracks [{
    type=MovieSceneAudioTrack
    display_name=Loop
    sections [
     {
      type=MovieSceneAudioSection
      end_frame=5959
      start_frame=5266
      channels [
       Volume {type=MovieSceneScriptingFloatChannel, default=1}
       Pitch {type=MovieSceneScriptingFloatChannel, default=1}
      ]
     }
     {
      type=MovieSceneAudioSection
      end_frame=6420
      start_frame=5266
      channels [
       Volume {type=MovieSceneScriptingFloatChannel, default=1}
       Pitch {type=MovieSceneScriptingFloatChannel, default=1}
      ]
     }
    ]
   }]
 }
 PistolCanary_RockSpireMonolith2 {
  id=#5
  possessed_class=PistolCanary_Monolith_Spire_C
  tracks [{
    type=MovieSceneAudioTrack
    display_name=Loop
    sections [
     {
      type=MovieSceneAudioSection
      end_frame=5958
      start_frame=5156
      channels [
       Volume {type=MovieSceneScriptingFloatChannel, default=1}
       Pitch {type=MovieSceneScriptingFloatChannel, default=1}
      ]
     }
     {
      type=MovieSceneAudioSection
      end_frame=6322
      start_frame=5156
      channels [
       Volume {type=MovieSceneScriptingFloatChannel, default=1}
       Pitch {type=MovieSceneScriptingFloatChannel, default=1}
      ]
     }
    ]
   }]
 }
 PistolCanary_RockSpireMonolith1 {
  id=#6
  possessed_class=PistolCanary_Monolith_Spire_C
  tracks [{
    type=MovieSceneAudioTrack
    display_name=Loop
    sections [
     {
      type=MovieSceneAudioSection
      end_frame=5957
      start_frame=5028
      channels [
       Volume {type=MovieSceneScriptingFloatChannel, default=0.6, keys [[5063.8275,1],[5396.3,0.6]]}
       Pitch {type=MovieSceneScriptingFloatChannel, default=1}
      ]
     }
     {
      type=MovieSceneAudioSection
      end_frame=6168
      start_frame=5028
      channels [
       Volume {type=MovieSceneScriptingFloatChannel, default=1}
       Pitch {type=MovieSceneScriptingFloatChannel, default=1}
      ]
     }
    ]
   }]
 }
 PistolCanary_AudioManager {
  id=#7
  possessed_class=PistolCanary_AudioManager_C
  tracks [
   {
    type=MovieSceneEventTrack
    display_name=AudioManagerEvents
    sections [{
      type=MovieSceneEventTriggerSection
      channels [None {
        type=MovieSceneScriptingEventChannel
        keys [
         [3619, "<Struct 'MovieSceneEvent' (0x852f8c04740) {}>"]
         [3723, "<Struct 'MovieSceneEvent' (0x852f8c0c260) {}>"]
         [5056, "<Struct 'MovieSceneEvent' (0x852f8c05780) {}>"]
         [6174.885, "<Struct 'MovieSceneEvent' (0x852f8c05960) {}>"]
         [6226.31, "<Struct 'MovieSceneEvent' (0x852f8c0d520) {}>"]
         [6410, "<Struct 'MovieSceneEvent' (0x852f8c046a0) {}>"]
         [6599, "<Struct 'MovieSceneEvent' (0x852f8c0ad20) {}>"]
         [7147, "<Struct 'MovieSceneEvent' (0x852f8c08f20) {}>"]
         [8392, "<Struct 'MovieSceneEvent' (0x852f8c0a5a0) {}>"]
        ]
       }]
     }]
   }
   {
    type=MovieSceneBoolTrack
    display_name=Event Mix ON
    sections [{type=MovieSceneBoolSection, channels [None {type=MovieSceneScriptingBoolChannel, default=false, keys [[3600,false],[3608,true]]}]}]
   }
   {
    type=MovieSceneIntegerTrack
    display_name=Music Phase
    sections [{
      type=MovieSceneIntegerSection
      channels [None {type=MovieSceneScriptingIntegerChannel, default=4, keys [[3557,0],[3599,0],[3727,1],[5062,2],[6409,3],[8331,4]]}]
     }]
   }
  ]
 }
 MaskFortress_CastleBase_01 {id=#8, possessed_class=BuildingLevelInstance, tracks []}
 PistolCanary_CinematicNPC_MaskMaker {
  id=#9
  possessed_class=PistolCanary_CinematicNPC_MaskMaker_C
  tracks [{
    type=MovieSceneEventTrack
    sections [{
      type=MovieSceneEventTriggerSection
      channels [None {type=MovieSceneScriptingEventChannel, const ["<Struct 'MovieSceneEvent' (0x852f8c0dc00) {}>", 1]}]
     }]
   }]
 }
 TentacleBoss {
  id=#10
  possessed_class=BGA_Tentacle_C
  tracks [{
    type=MovieSceneEventTrack
    sections [{
      type=MovieSceneEventTriggerSection
      channels [None {
        type=MovieSceneScriptingEventChannel
        keys [
         [6997, "<Struct 'MovieSceneEvent' (0x852f8c03f20) {}>"]
         [7153, "<Struct 'MovieSceneEvent' (0x852f8c0b860) {}>"]
         [8135, "<Struct 'MovieSceneEvent' (0x852f8c0d020) {}>"]
        ]
       }]
     }]
   }]
 }
]
display_rate {denominator=1, numerator=60}
tick_resolution {denominator=1, numerator=24000}
tracks [
 {
  type=MovieSceneAudioTrack
  display_name=MaskMaker
  sections [
   {
    type=MovieSceneAudioSection
    end_frame=7019
    start_frame=6092
    channels [
     Volume {type=MovieSceneScriptingFloatChannel, default=0.2, keys [[6180.11,1],[6198.26,0.2],[6238.505,1]]}
     Pitch {type=MovieSceneScriptingFloatChannel, default=1}
     AttachActor {type=MovieSceneScriptingActorReferenceChannel}
    ]
   }
   {
    type=MovieSceneAudioSection
    end_frame=8190
    start_frame=7993
    channels [
     Volume {type=MovieSceneScriptingFloatChannel, default=1}
     Pitch {type=MovieSceneScriptingFloatChannel, default=1}
     AttachActor {type=MovieSceneScriptingActorReferenceChannel}
    ]
   }
   {
    type=MovieSceneAudioSection
    end_frame=7019
    start_frame=5033
    channels [
     Volume {type=MovieSceneScriptingFloatChannel, default=1}
     Pitch {type=MovieSceneScriptingFloatChannel, default=1}
     AttachActor {type=MovieSceneScriptingActorReferenceChannel}
    ]
   }
   {
    type=MovieSceneAudioSection
    end_frame=13199
    start_frame=9125
    channels [
     Volume {type=MovieSceneScriptingFloatChannel, default=1}
     Pitch {type=MovieSceneScriptingFloatChannel, default=1}
     AttachActor {type=MovieSceneScriptingActorReferenceChannel}
    ]
   }
   {
    type=MovieSceneAudioSection
    end_frame=7972
    start_frame=3754
    channels [
     Volume {
      type=MovieSceneScriptingFloatChannel
      default=1
      thinned_from=7
      keys [[3763.5125,1],[5078.045,1],[5384.1175,0],[6263.09,1],[6459.6225,0.4],[6580.9325,0.65]]
     }
     Pitch {type=MovieSceneScriptingFloatChannel, default=1}
     AttachActor {type=MovieSceneScriptingActorReferenceChannel}
    ]
   }
  ]
 }
 {
  type=MovieSceneAudioTrack
  display_name=Music
  sections [
   {
    type=MovieSceneAudioSection
    end_frame=8413
    start_frame=6469
    channels [
     Volume {type=MovieSceneScriptingFloatChannel, default=1}
     Pitch {type=MovieSceneScriptingFloatChannel, default=1}
     AttachActor {type=MovieSceneScriptingActorReferenceChannel, default="<Struct 'MovieSceneObjectBindingID' (0x852f8ceffe0) {}>"}
    ]
   }
   {
    type=MovieSceneAudioSection
    end_frame=5062
    start_frame=3727
    channels [
     Volume {type=MovieSceneScriptingFloatChannel, default=1}
     Pitch {type=MovieSceneScriptingFloatChannel, default=1}
     AttachActor {type=MovieSceneScriptingActorReferenceChannel, default="<Struct 'MovieSceneObjectBindingID' (0x852f8ce4180) {}>"}
    ]
   }
  ]
 }
 {
  type=MovieSceneAudioTrack
  display_name=TentacleAudio
  sections [
   {
    type=MovieSceneAudioSection
    end_frame=8147
    start_frame=6926
    channels [
     Volume {type=MovieSceneScriptingFloatChannel, default=0.5, keys [[7399.3075,1],[7473.2125,0.5],[7577.5475,1]]}
     Pitch {type=MovieSceneScriptingFloatChannel, default=1}
     AttachActor {type=MovieSceneScriptingActorReferenceChannel}
    ]
   }
   {
    type=MovieSceneAudioSection
    end_frame=8925
    start_frame=6416
    channels [
     Volume {type=MovieSceneScriptingFloatChannel, default=1}
     Pitch {type=MovieSceneScriptingFloatChannel, default=1}
     AttachActor {type=MovieSceneScriptingActorReferenceChannel}
    ]
   }
   {
    type=MovieSceneAudioSection
    end_frame=8887
    start_frame=7656
    channels [
     Volume {type=MovieSceneScriptingFloatChannel, default=1}
     Pitch {type=MovieSceneScriptingFloatChannel, default=1}
     AttachActor {type=MovieSceneScriptingActorReferenceChannel}
    ]
   }
  ]
 }
 {
  type=MovieSceneAudioTrack
  display_name=Tentacle Head Whip
  sections [{
    type=MovieSceneAudioSection
    end_frame=8253
    start_frame=7930
    channels [
     Volume {type=MovieSceneScriptingFloatChannel, default=1}
     Pitch {type=MovieSceneScriptingFloatChannel, default=1}
     AttachActor {type=MovieSceneScriptingActorReferenceChannel}
    ]
   }]
 }
 {type=MovieSceneAudioTrack, display_name=House 5.1, sections []}
 {
  type=MovieSceneAudioTrack
  display_name=House
  sections [{
    type=MovieSceneAudioSection
    end_frame=7003
    start_frame=6204
    channels [
     Volume {type=MovieSceneScriptingFloatChannel, default=1.75, keys [[6375.8125,1],[6669.3675,1.75]]}
     Pitch {type=MovieSceneScriptingFloatChannel, default=1}
     AttachActor {type=MovieSceneScriptingActorReferenceChannel}
    ]
   }]
 }
 {
  type=MovieSceneAudioTrack
  display_name=MaskMakerDialogue
  sections [
   {
    type=MovieSceneAudioSection
    end_frame=4587
    start_frame=4264
    channels [
     Volume {type=MovieSceneScriptingFloatChannel, default=1}
     Pitch {type=MovieSceneScriptingFloatChannel, default=1}
     AttachActor {type=MovieSceneScriptingActorReferenceChannel}
    ]
   }
   {
    type=MovieSceneAudioSection
    end_frame=4717
    start_frame=4411
    channels [
     Volume {type=MovieSceneScriptingFloatChannel, default=1}
     Pitch {type=MovieSceneScriptingFloatChannel, default=1}
     AttachActor {type=MovieSceneScriptingActorReferenceChannel}
    ]
   }
   {
    type=MovieSceneAudioSection
    end_frame=5046
    start_frame=4727
    channels [
     Volume {type=MovieSceneScriptingFloatChannel, default=1}
     Pitch {type=MovieSceneScriptingFloatChannel, default=1}
     AttachActor {type=MovieSceneScriptingActorReferenceChannel}
    ]
   }
   {
    type=MovieSceneAudioSection
    end_frame=5310
    start_frame=5050
    channels [
     Volume {type=MovieSceneScriptingFloatChannel, default=1}
     Pitch {type=MovieSceneScriptingFloatChannel, default=1}
     AttachActor {type=MovieSceneScriptingActorReferenceChannel}
    ]
   }
   {
    type=MovieSceneAudioSection
    end_frame=5371
    start_frame=5184
    channels [
     Volume {type=MovieSceneScriptingFloatChannel, default=1}
     Pitch {type=MovieSceneScriptingFloatChannel, default=1}
     AttachActor {type=MovieSceneScriptingActorReferenceChannel}
    ]
   }
   {
    type=MovieSceneAudioSection
    end_frame=5584
    start_frame=5393
    channels [
     Volume {type=MovieSceneScriptingFloatChannel, default=1}
     Pitch {type=MovieSceneScriptingFloatChannel, default=1}
     AttachActor {type=MovieSceneScriptingActorReferenceChannel}
    ]
   }
   {
    type=MovieSceneAudioSection
    end_frame=5648
    start_frame=5556
    channels [
     Volume {type=MovieSceneScriptingFloatChannel, default=1}
     Pitch {type=MovieSceneScriptingFloatChannel, default=1}
     AttachActor {type=MovieSceneScriptingActorReferenceChannel}
    ]
   }
   {
    type=MovieSceneAudioSection
    end_frame=6124
    start_frame=5880
    channels [
     Volume {type=MovieSceneScriptingFloatChannel, default=1}
     Pitch {type=MovieSceneScriptingFloatChannel, default=1}
     AttachActor {type=MovieSceneScriptingActorReferenceChannel}
    ]
   }
   {
    type=MovieSceneAudioSection
    end_frame=5813
    start_frame=5641
    channels [
     Volume {type=MovieSceneScriptingFloatChannel, default=1}
     Pitch {type=MovieSceneScriptingFloatChannel, default=1}
     AttachActor {type=MovieSceneScriptingActorReferenceChannel}
    ]
   }
   {
    type=MovieSceneAudioSection
    end_frame=5827
    start_frame=5647
    channels [
     Volume {type=MovieSceneScriptingFloatChannel, default=1}
     Pitch {type=MovieSceneScriptingFloatChannel, default=1}
     AttachActor {type=MovieSceneScriptingActorReferenceChannel}
    ]
   }
   {
    type=MovieSceneAudioSection
    end_frame=6834
    start_frame=6468
    channels [
     Volume {type=MovieSceneScriptingFloatChannel, default=1}
     Pitch {type=MovieSceneScriptingFloatChannel, default=1}
     AttachActor {type=MovieSceneScriptingActorReferenceChannel}
    ]
   }
   {
    type=MovieSceneAudioSection
    end_frame=6310
    start_frame=6036
    channels [
     Volume {type=MovieSceneScriptingFloatChannel, default=1}
     Pitch {type=MovieSceneScriptingFloatChannel, default=1}
     AttachActor {type=MovieSceneScriptingActorReferenceChannel}
    ]
   }
   {
    type=MovieSceneAudioSection
    end_frame=8179
    start_frame=8004
    channels [
     Volume {type=MovieSceneScriptingFloatChannel, default=1}
     Pitch {type=MovieSceneScriptingFloatChannel, default=1}
     AttachActor {type=MovieSceneScriptingActorReferenceChannel}
    ]
   }
   {
    type=MovieSceneAudioSection
    end_frame=7816
    start_frame=7434
    channels [
     Volume {type=MovieSceneScriptingFloatChannel, default=1}
     Pitch {type=MovieSceneScriptingFloatChannel, default=1}
     AttachActor {type=MovieSceneScriptingActorReferenceChannel}
    ]
   }
  ]
 }
]
