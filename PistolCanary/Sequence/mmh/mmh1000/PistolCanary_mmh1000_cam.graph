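# Camera-effects Level Sequence for the PistolCanary mmh1000 sequence, timing camera shake, controller rumble and a brief late fade out and back.
# A spawned cameraForce Blueprint gets keyed shake classes (mprep, spire, castle, emerge) and small-to-large force feedback effects, fired by paired Lobby event trigger tracks.
# Two spawned cine cameras, a shake preview and a clips camera, play template sequences, the latter layering overlapping clips with keyed rotation multipliers to ramp each shake.
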LevelSequence /PistolCanary/Sequence/mmh/mmh1000/PistolCanary_mmh1000_cam
bindings [
 cameraShakePreview {
  id=#1
  possessed_class=CineCameraActor
  tracks [
   {
    type=MovieScene3DTransformTrack
    sections [{
      type=MovieScene3DTransformSection
      channels [
       Location.X {type=MovieSceneScriptingDoubleChannel, default=13518.643}
       Location.Y {type=MovieSceneScriptingDoubleChannel, default=82381.2165}
       Location.Z {type=MovieSceneScriptingDoubleChannel, default=4499.063}
       Rotation.X {type=MovieSceneScriptingDoubleChannel, default=0}
       Rotation.Y {type=MovieSceneScriptingDoubleChannel, default=13}
       Rotation.Z {type=MovieSceneScriptingDoubleChannel, default=-89.4}
       Scale.X {type=MovieSceneScriptingDoubleChannel, default=1}
       Scale.Y {type=MovieSceneScriptingDoubleChannel, default=1}
       Scale.Z {type=MovieSceneScriptingDoubleChannel, default=1}
      ]
     }]
   }
   {type=MovieSceneSpawnTrack, sections [{type=MovieSceneSpawnSection, channels [None {type=MovieSceneScriptingBoolChannel, default=true}]}]}
   {type=TemplateSequenceTrack, display_name=Template Animation, sections [{type=TemplateSequenceSection, end_frame=7016, start_frame=5961}]}
  ]
 }
 PistolCanary_cameraForce_BP {
  id=#2
  possessed_class=Events_Shared_cameraForce_BP_C
  tracks [
   {
    type=MovieSceneObjectPropertyTrack
    display_name=Force Effect
    sections [{
      type=MovieSceneObjectPropertySection
      channels [None {
        type=MovieSceneScriptingObjectPathChannel
        default="<Object '&Events_Shared_medium_FF' (0x8509a2598b0) Class 'ForceFeedbackEffect'>"
        keys [
         [0, "<Object '&Events_Shared_small_FF' (0x84c3c703480) Class 'ForceFeedbackEffect'>"]
         [6190, "<Object '/Events_Shared/ForceFeedback/Events_Shared_large_FF' (0x8528ea98dc0) Class 'ForceFeedbackEffect'>"]
         [6955, "<Object '&Events_Shared_small_FF' (0x84c3c703480) Class 'ForceFeedbackEffect'>"]
         [8177, "<Object '&Events_Shared_small_FF' (0x84c3c703480) Class 'ForceFeedbackEffect'>"]
        ]
       }]
     }]
   }
   {
    type=MovieSceneObjectPropertyTrack
    display_name=Cam Shake Base
    sections [{
      type=MovieSceneObjectPropertySection
      channels [None {
        type=MovieSceneScriptingObjectPathChannel
        default="<Object '&Events_Shared_00small_csb_C' (0x84c3c6a8a00) Class 'BlueprintGeneratedClass'>"
        keys [
         [
          0
          "<Object '&Events_Shared_00small_csb_C' (0x84c3c6a8a00) Class 'BlueprintGeneratedClass'>"
         ]
         [
          5051
          "<Object '/PistolCanary/Camera/PistolCanary_mmh1000_mprep_csb.PistolCanary_mmh1000_mprep_csb_C' (0x852d664ae00) Class 'BlueprintGeneratedClass'>"
         ]
         [
          5170
          "<Object '/PistolCanary/Camera/PistolCanary_mmh1000_spire_csb.PistolCanary_mmh1000_spire_csb_C' (0x852d664ba00) Class 'BlueprintGeneratedClass'>"
         ]
         [
          5907
          "<Object '/PistolCanary/Camera/PistolCanary_mmh1000_castle_csb.PistolCanary_mmh1000_castle_csb_C' (0x85295741200) Class 'BlueprintGeneratedClass'>"
         ]
         [
          6955
          "<Object '/PistolCanary/Camera/PistolCanary_spa1000_emerge1_csb.PistolCanary_spa1000_emerge1_csb_C' (0x852d664c600) Class 'BlueprintGeneratedClass'>"
         ]
         [8008,null]
         [
          8177
          "<Object '&PistolCanary_mmh1000_mprep_csb_C' (0x852d664ae00) Class 'BlueprintGeneratedClass'>"
         ]
        ]
       }]
     }]
   }
   {type=MovieSceneSpawnTrack, sections [{type=MovieSceneSpawnSection, channels [None {type=MovieSceneScriptingBoolChannel, default=true}]}]}
   {
    type=MovieScene3DTransformTrack
    sections [{
      type=MovieScene3DTransformSection
      channels [
       Location.X {type=MovieSceneScriptingDoubleChannel, default=14144.0716, const [14144.0713,1]}
       Location.Y {type=MovieSceneScriptingDoubleChannel, default=69971.8826, const [69971.8828,1]}
       Location.Z {type=MovieSceneScriptingDoubleChannel, default=0, const [6353.5273,1]}
       Rotation.X {type=MovieSceneScriptingDoubleChannel, default=0, const [0,1]}
       Rotation.Y {type=MovieSceneScriptingDoubleChannel, default=0, const [0,1]}
       Rotation.Z {type=MovieSceneScriptingDoubleChannel, default=0, const [0,1]}
       Scale.X {type=MovieSceneScriptingDoubleChannel, default=1, const [1,1]}
       Scale.Y {type=MovieSceneScriptingDoubleChannel, default=1, const [1,1]}
       Scale.Z {type=MovieSceneScriptingDoubleChannel, default=1, const [1,1]}
      ]
     }]
   }
   {
    type=MovieSceneEventTrack
    display_name=Cam Shake Lobby Event Trigger
    sections [{
      type=MovieSceneEventTriggerSection
      channels [None {
        type=MovieSceneScriptingEventChannel
        keys [
         [0, "<Struct 'MovieSceneEvent' (0x852f8c047e0) {}>"]
         [5051, "<Struct 'MovieSceneEvent' (0x852f8c0c3a0) {}>"]
         [5170, "<Struct 'MovieSceneEvent' (0x852f8c012c0) {}>"]
         [5277, "<Struct 'MovieSceneEvent' (0x852f8c0fe60) {}>"]
         [5383, "<Struct 'MovieSceneEvent' (0x852f8c019a0) {}>"]
         [5907, "<Struct 'MovieSceneEvent' (0x852f8c03b60) {}>"]
         [6955, "<Struct 'MovieSceneEvent' (0x852f8c0c300) {}>"]
         [8008, "<Struct 'MovieSceneEvent' (0x852f8c03520) {}>"]
         [8177, "<Struct 'MovieSceneEvent' (0x852f8c08520) {}>"]
        ]
       }]
     }]
   }
   {
    type=MovieSceneEventTrack
    display_name=Force Feedback Lobby Event Trigger
    sections [{
      type=MovieSceneEventTriggerSection
      channels [None {
        type=MovieSceneScriptingEventChannel
        keys [
         [0, "<Struct 'MovieSceneEvent' (0x852f8c03f20) {}>"]
         [5051, "<Struct 'MovieSceneEvent' (0x852f8c0dc00) {}>"]
         [5170, "<Struct 'MovieSceneEvent' (0x852f8c0a5a0) {}>"]
         [5277, "<Struct 'MovieSceneEvent' (0x852f8c08f20) {}>"]
         [5383, "<Struct 'MovieSceneEvent' (0x852f8c0ad20) {}>"]
         [6190, "<Struct 'MovieSceneEvent' (0x852f8c046a0) {}>"]
         [6955, "<Struct 'MovieSceneEvent' (0x852f8c0d520) {}>"]
         [8177, "<Struct 'MovieSceneEvent' (0x852f8c05960) {}>"]
        ]
       }]
     }]
   }
  ]
 }
 CameraComponent {
  id=#3
  parent_id=#1
  possessed_class=CineCameraComponent
  tracks [
   {
    type=MovieSceneFloatTrack
    display_name=Current Focal Length
    sections [{type=MovieSceneFloatSection, channels [None {type=MovieSceneScriptingFloatChannel, default=14}]}]
   }
   {
    type=MovieSceneFloatTrack
    display_name="Manual Focus Distance (Focus Settings)"
    sections [{type=MovieSceneFloatSection, channels [None {type=MovieSceneScriptingFloatChannel, default=100000}]}]
   }
   {
    type=MovieSceneFloatTrack
    display_name=Current Aperture
    sections [{type=MovieSceneFloatSection, channels [None {type=MovieSceneScriptingFloatChannel, default=2.8}]}]
   }
  ]
 }
 CameraComponent {
  id=#4
  parent_id=#5
  possessed_class=CineCameraComponent
  tracks [
   {
    type=MovieSceneFloatTrack
    display_name=Current Focal Length
    sections [{type=MovieSceneFloatSection, channels [None {type=MovieSceneScriptingFloatChannel, default=14}]}]
   }
   {
    type=MovieSceneFloatTrack
    display_name="Manual Focus Distance (Focus Settings)"
    sections [{type=MovieSceneFloatSection, channels [None {type=MovieSceneScriptingFloatChannel, default=100000}]}]
   }
   {
    type=MovieSceneFloatTrack
    display_name=Current Aperture
    sections [{type=MovieSceneFloatSection, channels [None {type=MovieSceneScriptingFloatChannel, default=2.8}]}]
   }
  ]
 }
 cameraShakeClips {
  id=#5
  possessed_class=CineCameraActor
  tracks [
   {type=MovieSceneSpawnTrack, sections [{type=MovieSceneSpawnSection, channels [None {type=MovieSceneScriptingBoolChannel, default=true}]}]}
   {
    type=TemplateSequenceTrack
    display_name=Template Animation
    sections [
     {
      type=TemplateSequenceSection
      end_frame=6887
      start_frame=5961
      channels ["Transform[Rotation]" {type=MovieSceneScriptingFloatChannel, default=0.35, keys [[5961,0],[6111,0.35],[6851.2925,0]]}]
     }
     {
      type=TemplateSequenceSection
      end_frame=6337
      start_frame=6188
      channels ["Transform[Rotation]" {type=MovieSceneScriptingFloatChannel, default=0, keys [[6188,0],[6199,0.5],[6333,0]]}]
     }
     {
      type=TemplateSequenceSection
      end_frame=6556
      start_frame=6275
      channels ["Transform[Rotation]" {type=MovieSceneScriptingFloatChannel, default=0, keys [[6277,0],[6348,0.4],[6543,0]]}]
     }
     {
      type=TemplateSequenceSection
      end_frame=7016
      start_frame=6494
      channels ["Transform[Rotation]" {type=MovieSceneScriptingFloatChannel, default=0.16, keys [[6493.9975,0],[6569.3775,0.16],[7014.115,0]]}]
     }
     {
      type=TemplateSequenceSection
      end_frame=7016
      start_frame=6494
      channels ["Transform[Rotation]" {type=MovieSceneScriptingFloatChannel, default=0, keys [[6499,0],[6569.3775,0.1],[6806,0]]}]
     }
    ]
   }
   {
    type=MovieScene3DTransformTrack
    sections [{
      type=MovieScene3DTransformSection
      channels [
       Location.X {type=MovieSceneScriptingDoubleChannel, default=8213.3988}
       Location.Y {type=MovieSceneScriptingDoubleChannel, default=84758.1381}
       Location.Z {type=MovieSceneScriptingDoubleChannel, default=4490.9446}
       Rotation.X {type=MovieSceneScriptingDoubleChannel, default=0.0131}
       Rotation.Y {type=MovieSceneScriptingDoubleChannel, default=6.31}
       Rotation.Z {type=MovieSceneScriptingDoubleChannel, default=-61.6367}
       Scale.X {type=MovieSceneScriptingDoubleChannel, default=1}
       Scale.Y {type=MovieSceneScriptingDoubleChannel, default=1}
       Scale.Z {type=MovieSceneScriptingDoubleChannel, default=1}
      ]
     }]
   }
  ]
 }
]
display_rate {denominator=1, numerator=60}
tick_resolution {denominator=1, numerator=24000}
tracks [{
  type=MovieSceneFadeTrack
  sections [{type=MovieSceneFadeSection, channels [None {type=MovieSceneScriptingFloatChannel, default=0, keys [[8331,0],[8348,1],[8356,0]]}]}]
 }]
