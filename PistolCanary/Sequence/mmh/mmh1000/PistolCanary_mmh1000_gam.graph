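# Gameplay-layer Level Sequence for PistolCanary shot mmh1000, which drives destruction, knockback and visibility changes around the castle and courtyard during a scripted event.
# Timed spawn tracks and event triggers run building and loot destroyers, while further events drive player-built structure destroyers and the Section1 sequence runner.
# Knockback volumes are keyed or attached, and a focus point moves between beats with its own events.
# The countdown hides early, and TentacleBoss stays hidden until later.
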LevelSequence /PistolCanary/Sequence/mmh/mmh1000/PistolCanary_mmh1000_gam
bindings [
 BP_PistolCanary_BuildingDestroyer2 {
  id=#1
  possessed_class=BP_PistolCanary_BuildingDestroyer_C
  tracks [
   {
    type=MovieSceneSpawnTrack
    sections [{
      type=MovieSceneSpawnSection
      channels [None {type=MovieSceneScriptingBoolChannel, default=false, keys [[6373.225,false],[6374.225,true],[8658,false]]}]
     }]
   }
   {
    type=MovieScene3DTransformTrack
    sections [{
      type=MovieScene3DTransformSection
      channels [
       Location.X {type=MovieSceneScriptingDoubleChannel, default=14359.8298}
       Location.Y {type=MovieSceneScriptingDoubleChannel, default=70375.2831}
       Location.Z {type=MovieSceneScriptingDoubleChannel, default=6481.6895, keys [[6390,6481.6895],[6419,5881.6143]]}
       Rotation.X {type=MovieSceneScriptingDoubleChannel, default=0}
       Rotation.Y {type=MovieSceneScriptingDoubleChannel, default=0}
       Rotation.Z {type=MovieSceneScriptingDoubleChannel, default=0}
       Scale.X {type=MovieSceneScriptingDoubleChannel, default=65}
       Scale.Y {type=MovieSceneScriptingDoubleChannel, default=75}
       Scale.Z {type=MovieSceneScriptingDoubleChannel, default=10}
      ]
     }]
   }
   {
    type=MovieSceneEventTrack
    sections [{
      type=MovieSceneEventTriggerSection
      channels [None {
        type=MovieSceneScriptingEventChannel
        keys [[6390, "<Struct 'MovieSceneEvent' (0x852f8c019a0) {}>"], [6420, "<Struct 'MovieSceneEvent' (0x852f8c03b60) {}>"]]
       }]
     }]
   }
  ]
 }
 BP_PistolCanary_KnockbackVolume_Castle {
  id=#2
  possessed_class=BP_PistolCanary_KnockbackVolume_C
  tracks [
   {
    type=MovieSceneEventTrack
    sections [{
      type=MovieSceneEventTriggerSection
      channels [None {type=MovieSceneScriptingEventChannel, const ["<Struct 'MovieSceneEvent' (0x852f8c05960) {}>", 1]}]
     }]
   }
   {
    type=MovieScene3DTransformTrack
    sections [{
      type=MovieScene3DTransformSection
      channels [
       Location.X {type=MovieSceneScriptingDoubleChannel, default=14424, const [14424,3]}
       Location.Y {type=MovieSceneScriptingDoubleChannel, default=70597, keys [[6171,70100],[6198,70100],[6242,70597]]}
       Location.Z {type=MovieSceneScriptingDoubleChannel, default=5500, keys [[6171,5500],[6198,3428],[6242,3428]]}
       Rotation.X {type=MovieSceneScriptingDoubleChannel, default=0}
       Rotation.Y {type=MovieSceneScriptingDoubleChannel, default=0}
       Rotation.Z {type=MovieSceneScriptingDoubleChannel, default=0}
       Scale.X {type=MovieSceneScriptingDoubleChannel, default=60, keys [[6171,60],[6180,137.5]]}
       Scale.Y {type=MovieSceneScriptingDoubleChannel, default=60, keys [[6171,60],[6180,137.5]]}
       Scale.Z {type=MovieSceneScriptingDoubleChannel, default=60, keys [[6171,60],[6180,137.5]]}
      ]
     }]
   }
  ]
 }
 BP_PistolCanary_KnockbackVolume_MaskMaker2 {
  id=#3
  possessed_class=BP_PistolCanary_KnockbackVolume_C
  tracks [
   {
    type=MovieSceneEventTrack
    sections [{
      type=MovieSceneEventTriggerSection
      channels [None {type=MovieSceneScriptingEventChannel, const ["<Struct 'MovieSceneEvent' (0x852f8c08de0) {}>", 1]}]
     }]
   }
   {type=MovieScene3DAttachTrack, sections [{type=MovieScene3DAttachSection, end_frame=8469, start_frame=3600}]}
  ]
 }
 BP_PistolCanary_PlayerBuiltStructureDestroyer_Castle {
  id=#4
  possessed_class=BP_PistolCanary_PlayerBuiltStructureDestroyer_C
  tracks [{
    type=MovieSceneEventTrack
    sections [{
      type=MovieSceneEventTriggerSection
      channels [None {
        type=MovieSceneScriptingEventChannel
        keys [
         [6198, "<Struct 'MovieSceneEvent' (0x852f8c08a20) {}>"]
         [7143, "<Struct 'MovieSceneEvent' (0x852f8c0a280) {}>"]
         [8690, "<Struct 'MovieSceneEvent' (0x852f8c04c40) {}>"]
        ]
       }]
     }]
   }]
 }
 TentacleBoss {
  id=#5
  possessed_class=PistolCanary_TentacleBoss_C
  tracks [{
    type=MovieSceneVisibilityTrack
    sections [{
      type=MovieSceneVisibilitySection
      channels [None {type=MovieSceneScriptingBoolChannel, default=true, keys [[3600,false],[6986,true],[6990,true],[6995,true]]}]
     }]
   }]
 }
 BP_PistolCanary_BuildingDestroyer3 {
  id=#6
  possessed_class=BP_PistolCanary_BuildingDestroyer_C
  tracks [
   {
    type=MovieSceneSpawnTrack
    sections [{
      type=MovieSceneSpawnSection
      channels [None {type=MovieSceneScriptingBoolChannel, default=false, keys [[6459,false],[6460,true],[8658,false]]}]
     }]
   }
   {
    type=MovieScene3DTransformTrack
    sections [{
      type=MovieScene3DTransformSection
      channels [
       Location.X {type=MovieSceneScriptingDoubleChannel, default=14588}
       Location.Y {type=MovieSceneScriptingDoubleChannel, default=70036}
       Location.Z {type=MovieSceneScriptingDoubleChannel, default=4288}
       Rotation.X {type=MovieSceneScriptingDoubleChannel, default=0}
       Rotation.Y {type=MovieSceneScriptingDoubleChannel, default=0}
       Rotation.Z {type=MovieSceneScriptingDoubleChannel, default=0}
       Scale.X {type=MovieSceneScriptingDoubleChannel, default=100}
       Scale.Y {type=MovieSceneScriptingDoubleChannel, default=100}
       Scale.Z {type=MovieSceneScriptingDoubleChannel, default=40}
      ]
     }]
   }
   {
    type=MovieSceneEventTrack
    sections [{
      type=MovieSceneEventTriggerSection
      channels [None {type=MovieSceneScriptingEventChannel, const ["<Struct 'MovieSceneEvent' (0x852f8c08700) {}>", 1]}]
     }]
   }
  ]
 }
 BP_PistolCanary_BuildingDestroyer1 {
  id=#7
  possessed_class=BP_PistolCanary_BuildingDestroyer_C
  tracks [
   {
    type=MovieSceneSpawnTrack
    sections [{
      type=MovieSceneSpawnSection
      channels [None {type=MovieSceneScriptingBoolChannel, default=false, keys [[6206.865,false],[6210,true],[8658,false],[12258,false]]}]
     }]
   }
   {
    type=MovieScene3DTransformTrack
    sections [{
      type=MovieScene3DTransformSection
      channels [
       Location.X {type=MovieSceneScriptingDoubleChannel, default=14588}
       Location.Y {type=MovieSceneScriptingDoubleChannel, default=70036}
       Location.Z {type=MovieSceneScriptingDoubleChannel, default=8161.6496}
       Rotation.X {type=MovieSceneScriptingDoubleChannel, default=0}
       Rotation.Y {type=MovieSceneScriptingDoubleChannel, default=0}
       Rotation.Z {type=MovieSceneScriptingDoubleChannel, default=0}
       Scale.X {type=MovieSceneScriptingDoubleChannel, default=100}
       Scale.Y {type=MovieSceneScriptingDoubleChannel, default=100}
       Scale.Z {type=MovieSceneScriptingDoubleChannel, default=37}
      ]
     }]
   }
   {
    type=MovieSceneEventTrack
    sections [{
      type=MovieSceneEventTriggerSection
      channels [None {type=MovieSceneScriptingEventChannel, const ["<Struct 'MovieSceneEvent' (0x852f8c0a780) {}>", 1]}]
     }]
   }
  ]
 }
 PistolCanary_BP_Countdown {
  id=#8
  possessed_class=BP_Generic_Game_Countdown_C
  tracks [{
    type=MovieSceneVisibilityTrack
    sections [{type=MovieSceneVisibilitySection, channels [None {type=MovieSceneScriptingBoolChannel, default=false, keys [[3600,true],[3601,false]]}]}]
   }]
 }
 PistolCanary_CinematicNPC_MaskMaker {id=#9, possessed_class=PistolCanary_CinematicNPC_MaskMaker_C, tracks []}
 BP_PistolCanary_PlayerBuiltStructureDestroyer_Courtyard {
  id=#10
  possessed_class=BP_PistolCanary_PlayerBuiltStructureDestroyer_C
  tracks [{
    type=MovieSceneEventTrack
    sections [{
      type=MovieSceneEventTriggerSection
      channels [None {
        type=MovieSceneScriptingEventChannel
        keys [[3641.25, "<Struct 'MovieSceneEvent' (0x852f8c0a5a0) {}>"], [6198, "<Struct 'MovieSceneEvent' (0x852f8c08f20) {}>"]]
       }]
     }]
   }]
 }
 BP_PistolCanary_LootDestroyer {
  id=#11
  possessed_class=BP_PistolCanary_LootDestroyer_C
  tracks [
   {
    type=MovieScene3DTransformTrack
    sections [{
      type=MovieScene3DTransformSection
      channels [
       Location.X {type=MovieSceneScriptingDoubleChannel, default=14276}
       Location.Y {type=MovieSceneScriptingDoubleChannel, default=70164}
       Location.Z {type=MovieSceneScriptingDoubleChannel, default=2629.6496}
       Rotation.X {type=MovieSceneScriptingDoubleChannel, default=0}
       Rotation.Y {type=MovieSceneScriptingDoubleChannel, default=0}
       Rotation.Z {type=MovieSceneScriptingDoubleChannel, default=0}
       Scale.X {type=MovieSceneScriptingDoubleChannel, default=150}
       Scale.Y {type=MovieSceneScriptingDoubleChannel, default=150}
       Scale.Z {type=MovieSceneScriptingDoubleChannel, default=200}
      ]
     }]
   }
   {
    type=MovieSceneSpawnTrack
    sections [{
      type=MovieSceneSpawnSection
      channels [None {type=MovieSceneScriptingBoolChannel, default=false, keys [[6150.865,false],[6154,true],[8658,false]]}]
     }]
   }
   {
    type=MovieSceneEventTrack
    sections [{
      type=MovieSceneEventTriggerSection
      channels [None {
        type=MovieSceneScriptingEventChannel
        keys [[6182, "<Struct 'MovieSceneEvent' (0x852f8c09b00) {}>"], [6714, "<Struct 'MovieSceneEvent' (0x852f8c00aa0) {}>"]]
       }]
     }]
   }
  ]
 }
 BP_PistolCanary_PlayerBuiltStructureDestroyer_FullPOI {
  id=#12
  possessed_class=BP_PistolCanary_PlayerBuiltStructureDestroyer_C
  tracks [{
    type=MovieSceneEventTrack
    sections [{
      type=MovieSceneEventTriggerSection
      channels [None {type=MovieSceneScriptingEventChannel, const ["<Struct 'MovieSceneEvent' (0x852f8c06fe0) {}>", 1]}]
     }]
   }]
 }
 BP_PistolCanary_FocusPoint {
  id=#13
  possessed_class=BP_PistolCanary_FocusPoint_C
  tracks [
   {
    type=MovieSceneEventTrack
    sections [{
      type=MovieSceneEventTriggerSection
      channels [None {
        type=MovieSceneScriptingEventChannel
        keys [
         [3637.675, "<Struct 'MovieSceneEvent' (0x852f8c0cee0) {}>"]
         [4254, "<Struct 'MovieSceneEvent' (0x852f8c0be00) {}>"]
         [6066, "<Struct 'MovieSceneEvent' (0x852f8c04f60) {}>"]
         [7670, "<Struct 'MovieSceneEvent' (0x852f8c044c0) {}>"]
         [8582, "<Struct 'MovieSceneEvent' (0x852f8c0b4a0) {}>"]
        ]
       }]
     }]
   }
   {
    type=MovieScene3DTransformTrack
    sections [{
      type=MovieScene3DTransformSection
      channels [
       Location.X {
        type=MovieSceneScriptingDoubleChannel
        default=14446.7494
        thinned_from=8
        keys [[4253.7225,14350.749],[5849,14094.749],[6171,14094.749],[6218,14388.0361],[7748.7625,14388.0361]]
       }
       Location.Y {
        type=MovieSceneScriptingDoubleChannel
        default=73107.7076
        keys [[4253.7225,73475.7109],[4470,75091.3125],[4615,75664.1719],[5849,76147.7109],[6171,76147.7109],[6218,70401.5781],[6385,70401.5781],[7655.7625,74162.6953],[8093.7625,74162.6953],[8270.7625,71654.6953]]
       }
       Location.Z {
        type=MovieSceneScriptingDoubleChannel
        default=7315
        keys [[4253.7225,4523.3271],[4470,4822.7568],[4615,5765.0142],[5849,6419.3271],[6171,6419.3271],[6218,7844.5029],[6385,7844.5029],[7748.7625,7347.4023],[8270.7625,7347.4023]]
       }
      ]
     }]
   }
  ]
 }
 BP_PistolCanary_SequenceRunner_Section1 {
  id=#14
  possessed_class=BP_PistolCanary_SequenceRunner_Section1_C
  tracks [{
    type=MovieSceneEventTrack
    sections [{
      type=MovieSceneEventTriggerSection
      channels [None {type=MovieSceneScriptingEventChannel, const ["<Struct 'MovieSceneEvent' (0x852f8c0a460) {}>", 1]}]
     }]
   }]
 }
]
display_rate {denominator=1, numerator=60}
tick_resolution {denominator=1, numerator=24000}
tracks [{
  type=MovieSceneEventTrack
  sections [{
    type=MovieSceneEventTriggerSection
    channels [None {
      type=MovieSceneScriptingEventChannel
      keys [
       [3637.675, "<Struct 'MovieSceneEvent' (0x852f8c04740) {}>"]
       [5170, "<Struct 'MovieSceneEvent' (0x852f8c0c260) {}>"]
       [8355, "<Struct 'MovieSceneEvent' (0x852f8c05780) {}>"]
       [8388, "<Struct 'MovieSceneEvent' (0x852f8c0b860) {}>"]
      ]
     }]
   }]
 }]
