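# Lighting Level Sequence for shot mmh1000 of the PistolCanary sequence, animating spotlights and day-sequence blending while hiding three jungle fog actors.
# A positioned DaySequenceModifierVolume ramps its user blend weight from zero to full, starting as the magenta inside spotlight fades after an initial intensity and radius flash.
# A second spotlight then rises, dips and recovers with color shifts and a narrowing cone.
# As it fades out, RectLight_pistolCanary, actually a spotlight, ramps up to a bright finish.
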LevelSequence /PistolCanary/Sequence/mmh/mmh1000/PistolCanary_mmh1000_lgt
bindings [
 DaySequenceModifier {
  id=#1
  parent_id=#2
  possessed_class=DaySequenceModifierComponent
  tracks [{
    type=MovieSceneFloatTrack
    display_name=User Blend Weight
    sections [{type=MovieSceneFloatSection, channels [None {type=MovieSceneScriptingFloatChannel, default=1, keys [[4737,0],[5713,1]]}]}]
   }]
 }
 LightComponent0 {
  id=#3
  parent_id=#4
  possessed_class=SpotLightComponent
  tracks [
   {
    type=MovieSceneFloatTrack
    display_name=Intensity
    sections [{
      type=MovieSceneFloatSection
      channels [None {type=MovieSceneScriptingFloatChannel, default=0, keys [[3749,7500],[3759,50000],[3799,7500],[4737,7500],[4810,0]]}]
     }]
   }
   {
    type=MovieSceneColorTrack
    display_name=Light Color
    sections [{
      type=MovieSceneColorSection
      channels [
       Color.R {type=MovieSceneScriptingFloatChannel, default=1}
       Color.G {type=MovieSceneScriptingFloatChannel, default=0.2705}
       Color.B {type=MovieSceneScriptingFloatChannel, default=0.9216}
       Color.A {type=MovieSceneScriptingFloatChannel, default=1}
      ]
     }]
   }
   {
    type=MovieSceneFloatTrack
    display_name=Attenuation Radius
    sections [{
      type=MovieSceneFloatSection
      channels [None {type=MovieSceneScriptingFloatChannel, default=3000, keys [[3749,3000],[3759,5000],[3799,3000]]}]
     }]
   }
  ]
 }
 BP_JungleFog {
  id=#5
  possessed_class=BP_JungleFog_C
  tracks [{
    type=MovieSceneVisibilityTrack
    sections [{type=MovieSceneVisibilitySection, channels [None {type=MovieSceneScriptingBoolChannel, default=false}]}]
   }]
 }
 SpotLight_PistolCanary_InsideSect1 {id=#4, possessed_class=SpotLight, tracks []}
 BP_JungleFog_LowHanging {
  id=#6
  possessed_class=BP_JungleFog_C
  tracks [{
    type=MovieSceneVisibilityTrack
    sections [{type=MovieSceneVisibilitySection, channels [None {type=MovieSceneScriptingBoolChannel, default=false}]}]
   }]
 }
 BP_JungleFog_LowHanging2 {
  id=#7
  possessed_class=BP_JungleFog_C
  tracks [{
    type=MovieSceneVisibilityTrack
    sections [{type=MovieSceneVisibilitySection, channels [None {type=MovieSceneScriptingBoolChannel, default=false}]}]
   }]
 }
 DaySequenceModifierVolume_PistolCanary {
  id=#2
  possessed_class=DaySequenceModifierVolume
  tracks [{
    type=MovieScene3DTransformTrack
    sections [{
      type=MovieScene3DTransformSection
      channels [
       Location.X {type=MovieSceneScriptingDoubleChannel, default=14456}
       Location.Y {type=MovieSceneScriptingDoubleChannel, default=77488}
       Location.Z {type=MovieSceneScriptingDoubleChannel, default=3688}
       Rotation.X {type=MovieSceneScriptingDoubleChannel, default=0}
       Rotation.Y {type=MovieSceneScriptingDoubleChannel, default=0}
       Rotation.Z {type=MovieSceneScriptingDoubleChannel, default=0}
       Scale.X {type=MovieSceneScriptingDoubleChannel, default=1}
       Scale.Y {type=MovieSceneScriptingDoubleChannel, default=1}
       Scale.Z {type=MovieSceneScriptingDoubleChannel, default=1}
      ]
     }]
   }]
 }
 LightComponent0 {
  id=#8
  parent_id=#9
  possessed_class=SpotLightComponent
  tracks [
   {
    type=MovieSceneFloatTrack
    display_name=Intensity
    sections [{
      type=MovieSceneFloatSection
      channels [None {
        type=MovieSceneScriptingFloatChannel
        default=25000
        thinned_from=7
        keys [[5068,0],[5895,25000],[6085,10000],[6148,10000],[6201,25000],[6423,0]]
       }]
     }]
   }
   {
    type=MovieSceneColorTrack
    display_name=Light Color
    sections [{
      type=MovieSceneColorSection
      channels [
       Color.R {type=MovieSceneScriptingFloatChannel, default=0.8148, const [1,4]}
       Color.G {type=MovieSceneScriptingFloatChannel, default=0.2582, keys [[5857,0.2159],[6081,0],[6167,0],[6201,0.0865]]}
       Color.B {type=MovieSceneScriptingFloatChannel, default=1, keys [[5857,0.8963],[6081,0.4621],[6167,0.2502],[6201,0.9473]]}
       Color.A {type=MovieSceneScriptingFloatChannel, default=1, const [1,4]}
      ]
     }]
   }
   {
    type=MovieSceneFloatTrack
    display_name=Outer Cone Angle
    sections [{type=MovieSceneFloatSection, channels [None {type=MovieSceneScriptingFloatChannel, default=35, keys [[5857,70],[6081,35]]}]}]
   }
  ]
 }
 RectLight_pistolCanary {id=#10, possessed_class=SpotLight, tracks []}
 LightComponent0 {
  id=#11
  parent_id=#10
  possessed_class=SpotLightComponent
  tracks [{
    type=MovieSceneFloatTrack
    display_name=Intensity
    sections [{type=MovieSceneFloatSection, channels [None {type=MovieSceneScriptingFloatChannel, default=1, keys [[6280,0],[6447,1],[7154,120000]]}]}]
   }]
 }
 SpotLight_PistolCanary_Sect1 {id=#9, possessed_class=SpotLight, tracks []}
]
display_rate {denominator=1, numerator=60}
tick_resolution {denominator=1, numerator=24000}
tracks []
