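# Level Sequence that times the PistolCanary tec layer for the mmh1000 sequence, covering its countdown, hero intro and comic-style intro effects.
# Event trigger tracks on the panel manager fire Dev Widget, Hero Intro, Comic Intro v2 and Comic Border events.
# Visibility tracks toggle the generic countdown and the comic intro post-process actor, which is placed and scaled here.
# Blend-weight keys fade in the comic volume and fade the intro post-process component in, then out.
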LevelSequence /PistolCanary/Sequence/mmh/mmh1000/PistolCanary_mmh1000_tec
bindings [
 PistolCanary_PanelManager {
  id=#1
  possessed_class=PistolCanary_PanelManager_C
  tracks [
   {
    type=MovieSceneEventTrack
    display_name=Dev Widget
    sections [{
      type=MovieSceneEventTriggerSection
      channels [None {type=MovieSceneScriptingEventChannel, const ["<Struct 'MovieSceneEvent' (0x852f8c00dc0) {}>", 1]}]
     }]
   }
   {
    type=MovieSceneEventTrack
    display_name=Dev Widget
    sections [{
      type=MovieSceneEventTriggerSection
      channels [None {type=MovieSceneScriptingEventChannel, const ["<Struct 'MovieSceneEvent' (0x852f8c00960) {}>", 1]}]
     }]
   }
   {
    type=MovieSceneEventTrack
    display_name=Hero Intro
    sections [{
      type=MovieSceneEventTriggerSection
      channels [None {
        type=MovieSceneScriptingEventChannel
        keys [
         [5519.7625, "<Struct 'MovieSceneEvent' (0x852f8c0fbe0) {}>"]
         [8095, "<Struct 'MovieSceneEvent' (0x852f8c06220) {}>"]
         [8348, "<Struct 'MovieSceneEvent' (0x852f8c0abe0) {}>"]
        ]
       }]
     }]
   }
   {
    type=MovieSceneEventTrack
    display_name=Comic Intro v2
    sections [{
      type=MovieSceneEventTriggerSection
      channels [None {type=MovieSceneScriptingEventChannel, const ["<Struct 'MovieSceneEvent' (0x852f8c05780) {}>", 1]}]
     }]
   }
   {
    type=MovieSceneEventTrack
    display_name=Comic Border
    sections [{
      type=MovieSceneEventTriggerSection
      channels [None {
        type=MovieSceneScriptingEventChannel
        keys [[8285, "<Struct 'MovieSceneEvent' (0x852f8c0be00) {}>"], [8346, "<Struct 'MovieSceneEvent' (0x852f8c0cee0) {}>"]]
       }]
     }]
   }
  ]
 }
 PistolCanary_BP_Countdown {
  id=#2
  possessed_class=BP_Generic_Game_Countdown_C
  tracks [{
    type=MovieSceneVisibilityTrack
    sections [{type=MovieSceneVisibilitySection, channels [None {type=MovieSceneScriptingBoolChannel, default=true, keys [[3599,true],[3600,false]]}]}]
   }]
 }
 PistolCanary_ComicEffect {
  id=#3
  possessed_class=PostProcessVolume
  tracks [{
    type=MovieSceneFloatTrack
    display_name=Blend Weight
    sections [{type=MovieSceneFloatSection, channels [None {type=MovieSceneScriptingFloatChannel, default=0, keys [[8077,0],[8113,1]]}]}]
   }]
 }
 PC_Comic_Intro_PP {
  id=#4
  possessed_class=PC_Comic_Intro_PP_C
  tracks [
   {
    type=MovieSceneVisibilityTrack
    sections [{type=MovieSceneVisibilitySection, channels [None {type=MovieSceneScriptingBoolChannel, default=true, keys [[3818,false],[8077,true]]}]}]
   }
   {
    type=MovieScene3DTransformTrack
    sections [{
      type=MovieScene3DTransformSection
      channels [
       Location.X {type=MovieSceneScriptingDoubleChannel, default=14488}
       Location.Y {type=MovieSceneScriptingDoubleChannel, default=72664}
       Location.Z {type=MovieSceneScriptingDoubleChannel, default=3724}
       Rotation.X {type=MovieSceneScriptingDoubleChannel, default=0}
       Rotation.Y {type=MovieSceneScriptingDoubleChannel, default=0}
       Rotation.Z {type=MovieSceneScriptingDoubleChannel, default=0}
       Scale.X {type=MovieSceneScriptingDoubleChannel, default=350}
       Scale.Y {type=MovieSceneScriptingDoubleChannel, default=350}
       Scale.Z {type=MovieSceneScriptingDoubleChannel, default=350}
      ]
     }]
   }
  ]
 }
 PostProcess {
  id=#5
  parent_id=#4
  possessed_class=PostProcessComponent
  tracks [{
    type=MovieSceneFloatTrack
    display_name=Blend Weight
    sections [{type=MovieSceneFloatSection, channels [None {type=MovieSceneScriptingFloatChannel, default=0, keys [[8077,0],[8188,1],[8348,1],[8376,0]]}]}]
   }]
 }
]
display_rate {denominator=1, numerator=60}
tick_resolution {denominator=1, numerator=24000}
tracks []
