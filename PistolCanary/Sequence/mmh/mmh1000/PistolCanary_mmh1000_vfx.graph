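# VFX Level Sequence for the PistolCanary mmh1000 cinematic, timing the Mask Maker's effects from spawn through energy drain, blast and tentacle emergence to smack impacts and roar.
# Spawn, visibility, particle and Niagara system tracks switch effects, some attached to bindings.
# Keyed Niagara parameters aim and swipe four drain beams, fade the spire beams and shape the charge.
# Material tracks drive the light cone and barrier collection alpha, alongside one subsequence.
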LevelSequence /PistolCanary/Sequence/mmh/mmh1000/PistolCanary_mmh1000_vfx
bindings [
 NS_PistolCanaryCD_Connecting_Spire_Glow_W {id=#1, possessed_class=NiagaraActor, tracks []}
 NiagaraComponent0 {
  id=#2
  parent_id=#3
  possessed_class=NiagaraComponent
  tracks [
   {
    type=MovieSceneNiagaraFloatParameterTrack
    display_name=Duration
    sections [{type=MovieSceneFloatSection, channels [None {type=MovieSceneScriptingFloatChannel, default=20}]}]
   }
   {
    type=MovieSceneNiagaraFloatParameterTrack
    display_name=FlareControl
    sections [{
      type=MovieSceneFloatSection
      channels [None {type=MovieSceneScriptingFloatChannel, default=1, keys [[4948,1],[5806,1],[5874,0],[6236,0],[6300,1]]}]
     }]
   }
   {
    type=MovieSceneNiagaraVectorParameterTrack
    display_name=HandPosOffset
    sections [{
      type=MovieSceneFloatVectorSection
      channels [
       Vector.X {type=MovieSceneScriptingFloatChannel, default=40, thinned_from=5, keys [[4936,0],[6040,0],[6051,40],[6111,100]]}
       Vector.Y {type=MovieSceneScriptingFloatChannel, default=-140, keys [[4936,0],[6040,0],[6051,-175],[6111,-165]]}
       Vector.Z {type=MovieSceneScriptingFloatChannel, default=-125, keys [[4936,0],[6040,0],[6051,-180],[6111,-160]]}
      ]
     }]
   }
   {
    type=MovieSceneParticleTrack
    display_name=FX System
    sections [{
      type=MovieSceneParticleSection
      channels [None {
        type=MovieSceneScriptingParticleChannel
        keys [
         [4914, "<ParticleKey.DEACTIVATE: 1>"]
         [4915, "<ParticleKey.ACTIVATE: 0>"]
         [6171, "<ParticleKey.ACTIVATE: 0>"]
         [6172, "<ParticleKey.DEACTIVATE: 1>"]
        ]
       }]
     }]
   }
  ]
 }
 NS_PistolCanary_MM_DrainEnergyLeft {
  id=#4
  possessed_class=NiagaraActor
  tracks [{
    type=MovieScene3DTransformTrack
    sections [{
      type=MovieScene3DTransformSection
      channels [
       Location.X {type=MovieSceneScriptingDoubleChannel, default=1476}
       Location.Y {type=MovieSceneScriptingDoubleChannel, default=70364}
       Location.Z {type=MovieSceneScriptingDoubleChannel, default=8932}
       Rotation.X {type=MovieSceneScriptingDoubleChannel, default=0}
       Rotation.Y {type=MovieSceneScriptingDoubleChannel, default=0}
       Rotation.Z {type=MovieSceneScriptingDoubleChannel, default=0}
       Scale.X {type=MovieSceneScriptingDoubleChannel, default=1}
       Scale.Y {type=MovieSceneScriptingDoubleChannel, default=1}
       Scale.Z {type=MovieSceneScriptingDoubleChannel, default=1}
      ]
     }]
   }]
 }
 Airflow_Spline {
  id=#5
  parent_id=#6
  possessed_class=StaticMeshComponent
  tracks [{type=MovieSceneComponentMaterialTrack, display_name=Material Slot: Material_0, sections []}]
 }
 NS_Pistol_Canary_MM_Charge {
  id=#3
  possessed_class=NiagaraActor
  tracks [
   {
    type=MovieScene3DTransformTrack
    sections [{
      type=MovieScene3DTransformSection
      channels [
       Location.X {type=MovieSceneScriptingDoubleChannel, default=0}
       Location.Y {type=MovieSceneScriptingDoubleChannel, default=0}
       Location.Z {type=MovieSceneScriptingDoubleChannel, default=0}
       Rotation.X {type=MovieSceneScriptingDoubleChannel, default=0}
       Rotation.Y {type=MovieSceneScriptingDoubleChannel, default=0}
       Rotation.Z {type=MovieSceneScriptingDoubleChannel, default=0}
       Scale.X {type=MovieSceneScriptingDoubleChannel, default=1}
       Scale.Y {type=MovieSceneScriptingDoubleChannel, default=1}
       Scale.Z {type=MovieSceneScriptingDoubleChannel, default=1}
      ]
     }]
   }
   {type=MovieScene3DAttachTrack, sections [{type=MovieScene3DAttachSection, end_frame=12085, start_frame=4538}]}
   {
    type=MovieSceneSpawnTrack
    sections [{
      type=MovieSceneSpawnSection
      channels [None {type=MovieSceneScriptingBoolChannel, default=false, keys [[4914,false],[4915,true],[6171,true],[6172,false]]}]
     }]
   }
  ]
 }
 NiagaraComponent0 {
  id=#7
  parent_id=#8
  possessed_class=NiagaraComponent
  tracks [
   {
    type=MovieSceneNiagaraVectorParameterTrack
    display_name=BeamEnd_0
    sections [{
      type=MovieSceneFloatVectorSection
      channels [
       Vector.X {type=MovieSceneScriptingFloatChannel, default=1703}
       Vector.Y {type=MovieSceneScriptingFloatChannel, default=70400}
       Vector.Z {type=MovieSceneScriptingFloatChannel, default=6150}
      ]
     }]
   }
   {
    type=MovieSceneNiagaraFloatParameterTrack
    display_name=BeamFade
    sections [{type=MovieSceneFloatSection, channels [None {type=MovieSceneScriptingFloatChannel, default=1, keys [[5398.67,1],[5497,0]]}]}]
   }
  ]
 }
 NiagaraComponent0 {
  id=#9
  parent_id=#10
  possessed_class=NiagaraComponent
  tracks [
   {type=MovieSceneNiagaraSystemTrack, sections [{type=MovieSceneNiagaraSystemSpawnSection, end_frame=5836, start_frame=5051}]}
   {
    type=MovieSceneNiagaraVectorParameterTrack
    display_name=Beam End
    sections [{
      type=MovieSceneFloatVectorSection
      channels [
       Vector.X {type=MovieSceneScriptingFloatChannel, default=14336, const [14336,1]}
       Vector.Y {type=MovieSceneScriptingFloatChannel, default=75900, const [75900,1]}
       Vector.Z {type=MovieSceneScriptingFloatChannel, default=6000, const [6000,1]}
      ]
     }]
   }
   {
    type=MovieSceneNiagaraFloatParameterTrack
    display_name=EnergyBeamSwipe
    sections [{
      type=MovieSceneFloatSection
      channels [None {type=MovieSceneScriptingFloatChannel, default=1, keys [[5051.4125,0],[5080,0.5],[5793.8675,0.5],[5836.3175,1]]}]
     }]
   }
  ]
 }
 NS_MM_Beam_Blast2 {
  id=#11
  possessed_class=NiagaraActor
  tracks [{
    type=MovieScene3DTransformTrack
    sections [{
      type=MovieScene3DTransformSection
      channels [
       Location.X {type=MovieSceneScriptingDoubleChannel, default=14128, const [14128,2]}
       Location.Y {type=MovieSceneScriptingDoubleChannel, default=75802, keys [[6085,75706],[6278,75986]]}
       Location.Z {type=MovieSceneScriptingDoubleChannel, default=6369, keys [[6085,6337],[6278,6409]]}
       Rotation.X {type=MovieSceneScriptingDoubleChannel, default=0, const [0,2]}
       Rotation.Y {type=MovieSceneScriptingDoubleChannel, default=0, const [0,2]}
       Rotation.Z {type=MovieSceneScriptingDoubleChannel, default=0, const [0,2]}
       Scale.X {type=MovieSceneScriptingDoubleChannel, default=1, const [1,2]}
       Scale.Y {type=MovieSceneScriptingDoubleChannel, default=1, const [1,2]}
       Scale.Z {type=MovieSceneScriptingDoubleChannel, default=1, const [1,2]}
      ]
     }]
   }]
 }
 NiagaraComponent0 {
  id=#12
  parent_id=#13
  possessed_class=NiagaraComponent
  tracks [
   {
    type=MovieSceneBoolTrack
    display_name=Auto Activate
    sections [{
      type=MovieSceneBoolSection
      channels [None {type=MovieSceneScriptingBoolChannel, default=false, keys [[3600,false],[6908,true],[7020,false]]}]
     }]
   }
   {type=MovieSceneNiagaraSystemTrack, sections [{type=MovieSceneNiagaraSystemSpawnSection, end_frame=7020, start_frame=6908}]}
  ]
 }
 NiagaraComponent0 {
  id=#14
  parent_id=#1
  possessed_class=NiagaraComponent
  tracks [
   {
    type=MovieSceneNiagaraFloatParameterTrack
    display_name=BeamFade
    sections [{type=MovieSceneFloatSection, channels [None {type=MovieSceneScriptingFloatChannel, default=1, keys [[5397,1],[5494,0]]}]}]
   }
   {
    type=MovieSceneParticleTrack
    display_name=FX System
    sections [{
      type=MovieSceneParticleSection
      channels [None {
        type=MovieSceneScriptingParticleChannel
        keys [
         [3541, "<ParticleKey.DEACTIVATE: 1>"]
         [3542, "<ParticleKey.ACTIVATE: 0>"]
         [5494, "<ParticleKey.ACTIVATE: 0>"]
         [5495, "<ParticleKey.DEACTIVATE: 1>"]
        ]
       }]
     }]
   }
  ]
 }
 NiagaraComponent0 {
  id=#15
  parent_id=#16
  possessed_class=NiagaraComponent
  tracks [{
    type=MovieSceneParticleTrack
    display_name=FX System
    sections [{
      type=MovieSceneParticleSection
      channels [None {
        type=MovieSceneScriptingParticleChannel
        keys [
         [6096, "<ParticleKey.DEACTIVATE: 1>"]
         [6097, "<ParticleKey.ACTIVATE: 0>"]
         [6350, "<ParticleKey.ACTIVATE: 0>"]
         [6351, "<ParticleKey.DEACTIVATE: 1>"]
        ]
       }]
     }]
   }]
 }
 NiagaraComponent0 {
  id=#17
  parent_id=#18
  possessed_class=NiagaraComponent
  tracks [
   {type=MovieScene3DAttachTrack, sections [{type=MovieScene3DAttachSection, end_frame=11583, start_frame=3569}]}
   {
    type=MovieScene3DTransformTrack
    sections [{
      type=MovieScene3DTransformSection
      channels [
       Location.X {type=MovieSceneScriptingDoubleChannel, default=-0.8712}
       Location.Y {type=MovieSceneScriptingDoubleChannel, default=29.0029}
       Location.Z {type=MovieSceneScriptingDoubleChannel, default=-481.2305}
       Rotation.X {type=MovieSceneScriptingDoubleChannel, default=0}
       Rotation.Y {type=MovieSceneScriptingDoubleChannel, default=0}
       Rotation.Z {type=MovieSceneScriptingDoubleChannel, default=0}
       Scale.X {type=MovieSceneScriptingDoubleChannel, default=1}
       Scale.Y {type=MovieSceneScriptingDoubleChannel, default=1}
       Scale.Z {type=MovieSceneScriptingDoubleChannel, default=1}
      ]
     }]
   }
   {
    type=MovieSceneParticleTrack
    display_name=FX System
    sections [{
      type=MovieSceneParticleSection
      channels [None {type=MovieSceneScriptingParticleChannel, keys [[3552, "<ParticleKey.TRIGGER: 2>"], [3688, "<ParticleKey.DEACTIVATE: 1>"]]}]
     }]
   }
  ]
 }
 NS_PistolCanary_MM_DrainEnergy4 {
  id=#19
  possessed_class=NiagaraActor
  tracks [{
    type=MovieScene3DTransformTrack
    sections [{
      type=MovieScene3DTransformSection
      channels [
       Location.X {type=MovieSceneScriptingDoubleChannel, default=26612}
       Location.Y {type=MovieSceneScriptingDoubleChannel, default=70420}
       Location.Z {type=MovieSceneScriptingDoubleChannel, default=8940}
       Rotation.X {type=MovieSceneScriptingDoubleChannel, default=0}
       Rotation.Y {type=MovieSceneScriptingDoubleChannel, default=0}
       Rotation.Z {type=MovieSceneScriptingDoubleChannel, default=0}
       Scale.X {type=MovieSceneScriptingDoubleChannel, default=1}
       Scale.Y {type=MovieSceneScriptingDoubleChannel, default=1}
       Scale.Z {type=MovieSceneScriptingDoubleChannel, default=1}
      ]
     }]
   }]
 }
 NS_PistolCanary_Spire_Beam_WtS {
  id=#20
  possessed_class=NiagaraActor
  tracks [
   {
    type=MovieSceneVisibilityTrack
    sections [{type=MovieSceneVisibilitySection, channels [None {type=MovieSceneScriptingBoolChannel, default=false, keys [[5496.9875,true],[5498,false]]}]}]
   }
   {
    type=MovieSceneNiagaraFloatParameterTrack
    display_name=BeamFade
    sections [{type=MovieSceneFloatSection, channels [None {type=MovieSceneScriptingFloatChannel, default=1, keys [[5400,1],[5497,0]]}]}]
   }
  ]
 }
 NS_MaskMaker_Cape_Spawn {
  id=#21
  possessed_class=NiagaraActor
  tracks [
   {
    type=MovieScene3DTransformTrack
    sections [{
      type=MovieScene3DTransformSection
      channels [
       Location.X {type=MovieSceneScriptingDoubleChannel, default=0}
       Location.Y {type=MovieSceneScriptingDoubleChannel, default=0}
       Location.Z {type=MovieSceneScriptingDoubleChannel, default=0}
       Rotation.X {type=MovieSceneScriptingDoubleChannel, default=0}
       Rotation.Y {type=MovieSceneScriptingDoubleChannel, default=0}
       Rotation.Z {type=MovieSceneScriptingDoubleChannel, default=0}
       Scale.X {type=MovieSceneScriptingDoubleChannel, default=1}
       Scale.Y {type=MovieSceneScriptingDoubleChannel, default=1}
       Scale.Z {type=MovieSceneScriptingDoubleChannel, default=1}
      ]
     }]
   }
   {type=MovieScene3DAttachTrack, sections [{type=MovieScene3DAttachSection, end_frame=12085, start_frame=1872}]}
   {
    type=MovieSceneSpawnTrack
    sections [{
      type=MovieSceneSpawnSection
      channels [None {type=MovieSceneScriptingBoolChannel, default=false, keys [[3513,false],[3676,true],[7948,false]]}]
     }]
   }
  ]
 }
 NS_Roar {
  id=#22
  possessed_class=NiagaraActor
  tracks [{
    type=MovieScene3DTransformTrack
    sections [{
      type=MovieScene3DTransformSection
      channels [
       Location.X {type=MovieSceneScriptingDoubleChannel, default=14950.3518, const [16242.3516,1]}
       Location.Y {type=MovieSceneScriptingDoubleChannel, default=70840, const [71420,1]}
       Location.Z {type=MovieSceneScriptingDoubleChannel, default=4187.9667, const [4787.9668,1]}
       Rotation.X {type=MovieSceneScriptingDoubleChannel, default=20, const [20,1]}
       Rotation.Y {type=MovieSceneScriptingDoubleChannel, default=0, const [0,1]}
       Rotation.Z {type=MovieSceneScriptingDoubleChannel, default=90, const [90,1]}
       Scale.X {type=MovieSceneScriptingDoubleChannel, default=25, const [25,1]}
       Scale.Y {type=MovieSceneScriptingDoubleChannel, default=25, const [25,1]}
       Scale.Z {type=MovieSceneScriptingDoubleChannel, default=25, const [25,1]}
      ]
     }]
   }]
 }
 NS_PistolCanary_MM_DrainEnergyFront {
  id=#10
  possessed_class=NiagaraActor
  tracks [{
    type=MovieScene3DTransformTrack
    sections [{
      type=MovieScene3DTransformSection
      channels [
       Location.X {type=MovieSceneScriptingDoubleChannel, default=14352, const [14352,1]}
       Location.Y {type=MovieSceneScriptingDoubleChannel, default=57128, const [57128,1]}
       Location.Z {type=MovieSceneScriptingDoubleChannel, default=9264, const [9264,1]}
       Rotation.X {type=MovieSceneScriptingDoubleChannel, default=0, const [0,1]}
       Rotation.Y {type=MovieSceneScriptingDoubleChannel, default=0, const [0,1]}
       Rotation.Z {type=MovieSceneScriptingDoubleChannel, default=0, const [0,1]}
       Scale.X {type=MovieSceneScriptingDoubleChannel, default=1, const [1,1]}
       Scale.Y {type=MovieSceneScriptingDoubleChannel, default=1, const [1,1]}
       Scale.Z {type=MovieSceneScriptingDoubleChannel, default=1, const [1,1]}
      ]
     }]
   }]
 }
 NS_PistolCanary_Spire_Beam_StE {
  id=#23
  possessed_class=NiagaraActor
  tracks [{
    type=MovieSceneVisibilityTrack
    sections [{type=MovieSceneVisibilitySection, channels [None {type=MovieSceneScriptingBoolChannel, default=false, keys [[5493,true],[5494,false]]}]}]
   }]
 }
 NS_PistolCanaryCD_Connecting_Spire_Glow_E {id=#24, possessed_class=NiagaraActor, tracks []}
 NiagaraComponent0 {
  id=#25
  parent_id=#19
  possessed_class=NiagaraComponent
  tracks [
   {type=MovieSceneNiagaraSystemTrack, sections [{type=MovieSceneNiagaraSystemSpawnSection, end_frame=5836, start_frame=5262}]}
   {
    type=MovieSceneNiagaraVectorParameterTrack
    display_name=Beam End
    sections [{
      type=MovieSceneFloatVectorSection
      channels [
       Vector.X {type=MovieSceneScriptingFloatChannel, default=14313, const [14336,1]}
       Vector.Y {type=MovieSceneScriptingFloatChannel, default=77908, const [75900,1]}
       Vector.Z {type=MovieSceneScriptingFloatChannel, default=7100, const [6000,1]}
      ]
     }]
   }
   {
    type=MovieSceneNiagaraFloatParameterTrack
    display_name=EnergyBeamSwipe
    sections [{
      type=MovieSceneFloatSection
      channels [None {type=MovieSceneScriptingFloatChannel, default=1, keys [[5262.4025,0],[5291,0.5],[5787.975,0.5],[5836,1]]}]
     }]
   }
  ]
 }
 NiagaraComponent0 {
  id=#26
  parent_id=#27
  possessed_class=NiagaraComponent
  tracks [{
    type=MovieSceneEventTrack
    sections [{
      type=MovieSceneEventTriggerSection
      channels [None {type=MovieSceneScriptingEventChannel, const ["<Struct 'MovieSceneEvent' (0x852f8c0cb20) {}>", 1]}]
     }]
   }]
 }
 NS_PC_MaskMaker_Smack_MountainImpact {
  id=#27
  possessed_class=NiagaraActor
  tracks [
   {
    type=MovieSceneSpawnTrack
    sections [{type=MovieSceneSpawnSection, channels [None {type=MovieSceneScriptingBoolChannel, default=false, keys [[3536,false],[8050,true]]}]}]
   }
   {
    type=MovieScene3DTransformTrack
    sections [{
      type=MovieScene3DTransformSection
      channels [
       Location.X {type=MovieSceneScriptingDoubleChannel, default=-33023.8555, const [-32191.8555,1]}
       Location.Y {type=MovieSceneScriptingDoubleChannel, default=83744.7718, const [83404.7734,1]}
       Location.Z {type=MovieSceneScriptingDoubleChannel, default=16686, const [17326,1]}
       Rotation.X {type=MovieSceneScriptingDoubleChannel, default=0, const [0,1]}
       Rotation.Y {type=MovieSceneScriptingDoubleChannel, default=30, const [30,1]}
       Rotation.Z {type=MovieSceneScriptingDoubleChannel, default=-50, const [-50,1]}
       Scale.X {type=MovieSceneScriptingDoubleChannel, default=1, const [1,1]}
       Scale.Y {type=MovieSceneScriptingDoubleChannel, default=1, const [1,1]}
       Scale.Z {type=MovieSceneScriptingDoubleChannel, default=1, const [1,1]}
      ]
     }]
   }
  ]
 }
 NiagaraComponent0 {
  id=#28
  parent_id=#20
  possessed_class=NiagaraComponent
  tracks [
   {
    type=MovieSceneNiagaraVectorParameterTrack
    display_name=BeamEnd_0
    sections [{
      type=MovieSceneFloatVectorSection
      channels [
       Vector.X {type=MovieSceneScriptingFloatChannel, default=14500}
       Vector.Y {type=MovieSceneScriptingFloatChannel, default=86600}
       Vector.Z {type=MovieSceneScriptingFloatChannel, default=6100}
      ]
     }]
   }
   {
    type=MovieSceneNiagaraFloatParameterTrack
    display_name=BeamFade
    sections [{type=MovieSceneFloatSection, channels [None {type=MovieSceneScriptingFloatChannel, default=1, keys [[5395.6575,1],[5496.9875,0]]}]}]
   }
  ]
 }
 NS_PistolCanaryCD_Connecting_Spire_Glow_N {id=#29, possessed_class=NiagaraActor, tracks []}
 BP_PistolCanary_LightBeam {
  id=#6
  possessed_class=null
  tracks [{
    type=MovieSceneVisibilityTrack
    sections [{type=MovieSceneVisibilitySection, channels [None {type=MovieSceneScriptingBoolChannel, default=true, keys [[3780,true],[3782,false]]}]}]
   }]
 }
 NS_PC_Temple_Shockwave {
  id=#30
  possessed_class=NiagaraActor
  tracks [
   {
    type=MovieSceneSpawnTrack
    sections [{type=MovieSceneSpawnSection, channels [None {type=MovieSceneScriptingBoolChannel, default=false, keys [[3538,false],[6125,true]]}]}]
   }
   {
    type=MovieScene3DTransformTrack
    sections [{
      type=MovieScene3DTransformSection
      channels [
       Location.X {type=MovieSceneScriptingDoubleChannel, default=14328}
       Location.Y {type=MovieSceneScriptingDoubleChannel, default=70128}
       Location.Z {type=MovieSceneScriptingDoubleChannel, default=7446.264}
       Rotation.X {type=MovieSceneScriptingDoubleChannel, default=0}
       Rotation.Y {type=MovieSceneScriptingDoubleChannel, default=0}
       Rotation.Z {type=MovieSceneScriptingDoubleChannel, default=0}
       Scale.X {type=MovieSceneScriptingDoubleChannel, default=1}
       Scale.Y {type=MovieSceneScriptingDoubleChannel, default=1}
       Scale.Z {type=MovieSceneScriptingDoubleChannel, default=1}
      ]
     }]
   }
  ]
 }
 NiagaraComponent0 {
  id=#31
  parent_id=#32
  possessed_class=NiagaraComponent
  tracks [
   {type=MovieSceneNiagaraSystemTrack, sections [{type=MovieSceneNiagaraSystemSpawnSection, end_frame=5836, start_frame=5156}]}
   {
    type=MovieSceneNiagaraVectorParameterTrack
    display_name=Beam End
    sections [{
      type=MovieSceneFloatVectorSection
      channels [
       Vector.X {type=MovieSceneScriptingFloatChannel, default=14313, const [14336,1]}
       Vector.Y {type=MovieSceneScriptingFloatChannel, default=77908, const [75900,1]}
       Vector.Z {type=MovieSceneScriptingFloatChannel, default=7100, const [6000,1]}
      ]
     }]
   }
   {
    type=MovieSceneNiagaraFloatParameterTrack
    display_name=EnergyBeamSwipe
    sections [{
      type=MovieSceneFloatSection
      channels [None {type=MovieSceneScriptingFloatChannel, default=1, keys [[5156.0775,0],[5184.675,0.5],[5787.975,0.5],[5836,1]]}]
     }]
   }
  ]
 }
 NS_MaskMaker_IntroSpawn {
  id=#18
  possessed_class=NiagaraActor
  tracks [{
    type=MovieScene3DTransformTrack
    sections [{
      type=MovieScene3DTransformSection
      channels [
       Location.X {type=MovieSceneScriptingDoubleChannel, default=4}
       Location.Y {type=MovieSceneScriptingDoubleChannel, default=4}
       Location.Z {type=MovieSceneScriptingDoubleChannel, default=-83}
       Rotation.X {type=MovieSceneScriptingDoubleChannel, default=0}
       Rotation.Y {type=MovieSceneScriptingDoubleChannel, default=0}
       Rotation.Z {type=MovieSceneScriptingDoubleChannel, default=0}
       Scale.X {type=MovieSceneScriptingDoubleChannel, default=1}
       Scale.Y {type=MovieSceneScriptingDoubleChannel, default=1}
       Scale.Z {type=MovieSceneScriptingDoubleChannel, default=1}
      ]
     }]
   }]
 }
 NS_PistolCanary_BeamEnergyUp {
  id=#16
  possessed_class=NiagaraActor
  tracks [{
    type=MovieScene3DTransformTrack
    sections [{
      type=MovieScene3DTransformSection
      channels [
       Location.X {type=MovieSceneScriptingDoubleChannel, default=14348}
       Location.Y {type=MovieSceneScriptingDoubleChannel, default=69801}
       Location.Z {type=MovieSceneScriptingDoubleChannel, default=3602}
       Rotation.X {type=MovieSceneScriptingDoubleChannel, default=0}
       Rotation.Y {type=MovieSceneScriptingDoubleChannel, default=0}
       Rotation.Z {type=MovieSceneScriptingDoubleChannel, default=0}
       Scale.X {type=MovieSceneScriptingDoubleChannel, default=3}
       Scale.Y {type=MovieSceneScriptingDoubleChannel, default=3}
       Scale.Z {type=MovieSceneScriptingDoubleChannel, default=3}
      ]
     }]
   }]
 }
 S_Billboard_LightCone {
  id=#33
  possessed_class=FortStaticMeshActor
  tracks [
   {
    type=MovieScene3DTransformTrack
    sections [{
      type=MovieScene3DTransformSection
      channels [
       Location.X {type=MovieSceneScriptingDoubleChannel, default=14436, const [14436,1]}
       Location.Y {type=MovieSceneScriptingDoubleChannel, default=69812, const [69812,1]}
       Location.Z {type=MovieSceneScriptingDoubleChannel, default=3912, const [3912,1]}
       Rotation.X {type=MovieSceneScriptingDoubleChannel, default=-180, const [-180,1]}
       Rotation.Y {type=MovieSceneScriptingDoubleChannel, default=0, const [0,1]}
       Rotation.Z {type=MovieSceneScriptingDoubleChannel, default=0, const [0,1]}
       Scale.X {type=MovieSceneScriptingDoubleChannel, default=200, keys [[6181,20],[6195,0],[6700,200]]}
       Scale.Y {type=MovieSceneScriptingDoubleChannel, default=200, keys [[6181,0],[6191,20],[6700,200]]}
       Scale.Z {type=MovieSceneScriptingDoubleChannel, default=120, keys [[6181,0],[6191,75],[6700,120]]}
      ]
     }]
   }
   {
    type=MovieSceneVisibilityTrack
    sections [{type=MovieSceneVisibilitySection, channels [None {type=MovieSceneScriptingBoolChannel, default=true, keys [[3589,false],[6160,true]]}]}]
   }
  ]
 }
 NiagaraComponent0 {id=#34, parent_id=#35, possessed_class=NiagaraComponent, tracks []}
 NS_MaskFortress_Portal_Day3 {
  id=#36
  parent_id=#37
  possessed_class=NiagaraComponent
  tracks [
   {
    type=MovieSceneVisibilityTrack
    sections [{type=MovieSceneVisibilitySection, channels [None {type=MovieSceneScriptingBoolChannel, default=true, keys [[3513,true],[6176,false]]}]}]
   }
   {
    type=MovieSceneBoolTrack
    display_name=Auto Activate
    sections [{type=MovieSceneBoolSection, channels [None {type=MovieSceneScriptingBoolChannel, default=false, keys [[3513,true],[6176,false]]}]}]
   }
  ]
 }
 NS_PistolCanaryCD_Connecting_Spire_Glow_S {id=#38, possessed_class=NiagaraActor, tracks []}
 NS_PistolCanary_Tentacle_Emerge {
  id=#13
  possessed_class=NiagaraActor
  tracks [{
    type=MovieScene3DTransformTrack
    sections [{
      type=MovieScene3DTransformSection
      channels [
       Location.X {type=MovieSceneScriptingDoubleChannel, default=14328}
       Location.Y {type=MovieSceneScriptingDoubleChannel, default=70128}
       Location.Z {type=MovieSceneScriptingDoubleChannel, default=3896}
       Rotation.X {type=MovieSceneScriptingDoubleChannel, default=0}
       Rotation.Y {type=MovieSceneScriptingDoubleChannel, default=0}
       Rotation.Z {type=MovieSceneScriptingDoubleChannel, default=0}
       Scale.X {type=MovieSceneScriptingDoubleChannel, default=1}
       Scale.Y {type=MovieSceneScriptingDoubleChannel, default=1}
       Scale.Z {type=MovieSceneScriptingDoubleChannel, default=1}
      ]
     }]
   }]
 }
 NS_PistolCanary_Spire_Beam_EtN {
  id=#39
  possessed_class=NiagaraActor
  tracks [{
    type=MovieSceneVisibilityTrack
    sections [{
      type=MovieSceneVisibilitySection
      channels [None {type=MovieSceneScriptingBoolChannel, default=false, keys [[5488.8925,true],[5489.9075,false]]}]
     }]
   }]
 }
 NS_PistolCanary_MM_DrainEnergyBack {
  id=#32
  possessed_class=NiagaraActor
  tracks [{
    type=MovieScene3DTransformTrack
    sections [{
      type=MovieScene3DTransformSection
      channels [
       Location.X {type=MovieSceneScriptingDoubleChannel, default=14292}
       Location.Y {type=MovieSceneScriptingDoubleChannel, default=86968}
       Location.Z {type=MovieSceneScriptingDoubleChannel, default=8916}
       Rotation.X {type=MovieSceneScriptingDoubleChannel, default=0}
       Rotation.Y {type=MovieSceneScriptingDoubleChannel, default=0}
       Rotation.Z {type=MovieSceneScriptingDoubleChannel, default=0}
       Scale.X {type=MovieSceneScriptingDoubleChannel, default=1}
       Scale.Y {type=MovieSceneScriptingDoubleChannel, default=1}
       Scale.Z {type=MovieSceneScriptingDoubleChannel, default=1}
      ]
     }]
   }]
 }
 NiagaraComponent0 {id=#40, parent_id=#21, possessed_class=NiagaraComponent, tracks []}
 NiagaraComponent0 {
  id=#41
  parent_id=#38
  possessed_class=NiagaraComponent
  tracks [
   {
    type=MovieSceneNiagaraFloatParameterTrack
    display_name=BeamFade
    sections [{type=MovieSceneFloatSection, channels [None {type=MovieSceneScriptingFloatChannel, default=1, keys [[5397,1],[5494,0]]}]}]
   }
   {
    type=MovieSceneParticleTrack
    display_name=FX System
    sections [{
      type=MovieSceneParticleSection
      channels [None {
        type=MovieSceneScriptingParticleChannel
        keys [
         [3541, "<ParticleKey.DEACTIVATE: 1>"]
         [3542, "<ParticleKey.ACTIVATE: 0>"]
         [5494, "<ParticleKey.ACTIVATE: 0>"]
         [5495, "<ParticleKey.DEACTIVATE: 1>"]
        ]
       }]
     }]
   }
  ]
 }
 NS_PistolCanary_Tentacle_GroundPortal_01 {
  id=#42
  possessed_class=NiagaraActor
  tracks [{
    type=MovieScene3DTransformTrack
    sections [{
      type=MovieScene3DTransformSection
      channels [
       Location.X {type=MovieSceneScriptingDoubleChannel, default=14564, const [14328,1]}
       Location.Y {type=MovieSceneScriptingDoubleChannel, default=70128, const [70128,1]}
       Location.Z {type=MovieSceneScriptingDoubleChannel, default=4760, const [4760,2]}
       Rotation.X {type=MovieSceneScriptingDoubleChannel, default=0, const [0,1]}
       Rotation.Y {type=MovieSceneScriptingDoubleChannel, default=0, const [0,1]}
       Rotation.Z {type=MovieSceneScriptingDoubleChannel, default=0, const [0,1]}
       Scale.X {type=MovieSceneScriptingDoubleChannel, default=1.1, keys [[6125,0],[6654,1.1]]}
       Scale.Y {type=MovieSceneScriptingDoubleChannel, default=1.1, keys [[6125,0],[6654,1.1]]}
       Scale.Z {type=MovieSceneScriptingDoubleChannel, default=1.1, keys [[6125,0],[6654,1.1]]}
      ]
     }]
   }]
 }
 NiagaraComponent0 {
  id=#43
  parent_id=#44
  possessed_class=NiagaraComponent
  tracks [{
    type=MovieSceneParticleTrack
    display_name=FX System
    sections [{
      type=MovieSceneParticleSection
      channels [None {
        type=MovieSceneScriptingParticleChannel
        keys [
         [7926, "<ParticleKey.DEACTIVATE: 1>"]
         [7927, "<ParticleKey.ACTIVATE: 0>"]
         [8006, "<ParticleKey.ACTIVATE: 0>"]
         [8007, "<ParticleKey.DEACTIVATE: 1>"]
        ]
       }]
     }]
   }]
 }
 NS_MaskFortress_Portal_Day2 {
  id=#45
  parent_id=#37
  possessed_class=NiagaraComponent
  tracks [
   {
    type=MovieSceneVisibilityTrack
    sections [{type=MovieSceneVisibilitySection, channels [None {type=MovieSceneScriptingBoolChannel, default=false, const [false,1]}]}]
   }
   {
    type=MovieSceneBoolTrack
    display_name=Auto Activate
    sections [{type=MovieSceneBoolSection, channels [None {type=MovieSceneScriptingBoolChannel, default=false, const [false,1]}]}]
   }
  ]
 }
 BP_MaskMaker_Portal_PistolCanary {
  id=#37
  possessed_class=BP_MaskMaker_Portal_PistolCanary_C
  tracks [{
    type=MovieSceneVisibilityTrack
    sections [{type=MovieSceneVisibilitySection, channels [None {type=MovieSceneScriptingBoolChannel, default=true, keys [[3513,true],[6176,false]]}]}]
   }]
 }
 SM_PistolCanary_Cone {
  id=#46
  parent_id=#37
  possessed_class=StaticMeshComponent
  tracks [{
    type=MovieSceneVisibilityTrack
    sections [{type=MovieSceneVisibilitySection, channels [None {type=MovieSceneScriptingBoolChannel, default=true, keys [[3513,true],[6176,false]]}]}]
   }]
 }
 StaticMeshComponent0 {
  id=#47
  parent_id=#33
  possessed_class=StaticMeshComponent
  tracks [
   {
    type=MovieScenePrimitiveMaterialTrack
    display_name=Material Slot: M_Billboard_SpotLight
    sections [{
      type=MovieScenePrimitiveMaterialSection
      channels [None {
        type=MovieSceneScriptingObjectPathChannel
        default="<Object '/PistolCanary/Assets/FX/GroundPortal/MI_PistolCanary_LightCone' (0x852d619b800) Class 'MaterialInstanceConstant'>"
       }]
     }]
   }
   {
    type=MovieSceneComponentMaterialTrack
    display_name=Material Slot: M_Billboard_SpotLight
    sections [{
      type=MovieSceneComponentMaterialParameterSection
      end_frame=12890
      start_frame=4405
      channels [
       Brightness {type=MovieSceneScriptingFloatChannel, default=0.25, const [0.2,1]}
       CenterHole {type=MovieSceneScriptingFloatChannel, default=0.1, keys [[6882,0.01],[7032,0.1]]}
       FadeSensitivity {type=MovieSceneScriptingFloatChannel, default=200, keys [[7143,4],[7366,200]]}
      ]
     }]
   }
  ]
 }
 NS Pistol Canary MM Aura Loop {
  id=#35
  possessed_class=NiagaraActor
  tracks [
   {
    type=MovieScene3DTransformTrack
    sections [{
      type=MovieScene3DTransformSection
      channels [
       Location.X {type=MovieSceneScriptingDoubleChannel, default=0}
       Location.Y {type=MovieSceneScriptingDoubleChannel, default=0}
       Location.Z {type=MovieSceneScriptingDoubleChannel, default=0}
       Rotation.X {type=MovieSceneScriptingDoubleChannel, default=0}
       Rotation.Y {type=MovieSceneScriptingDoubleChannel, default=0}
       Rotation.Z {type=MovieSceneScriptingDoubleChannel, default=0}
       Scale.X {type=MovieSceneScriptingDoubleChannel, default=1}
       Scale.Y {type=MovieSceneScriptingDoubleChannel, default=1}
       Scale.Z {type=MovieSceneScriptingDoubleChannel, default=1}
      ]
     }]
   }
   {type=MovieScene3DAttachTrack, sections [{type=MovieScene3DAttachSection, end_frame=11656, start_frame=3642}]}
   {
    type=MovieSceneSpawnTrack
    sections [{
      type=MovieSceneSpawnSection
      channels [None {type=MovieSceneScriptingBoolChannel, default=false, keys [[3671,true],[5176,false],[6383,true],[8177,false]]}]
     }]
   }
  ]
 }
 NiagaraComponent0 {
  id=#48
  parent_id=#4
  possessed_class=NiagaraComponent
  tracks [
   {type=MovieSceneNiagaraSystemTrack, sections [{type=MovieSceneNiagaraSystemSpawnSection, end_frame=5836, start_frame=4936}]}
   {
    type=MovieSceneNiagaraVectorParameterTrack
    display_name=Beam End
    sections [{
      type=MovieSceneFloatVectorSection
      channels [
       Vector.X {type=MovieSceneScriptingFloatChannel, default=14313, const [14336,1]}
       Vector.Y {type=MovieSceneScriptingFloatChannel, default=77908, const [75900,1]}
       Vector.Z {type=MovieSceneScriptingFloatChannel, default=7100, const [6000,1]}
      ]
     }]
   }
   {
    type=MovieSceneNiagaraFloatParameterTrack
    display_name=EnergyBeamSwipe
    sections [{
      type=MovieSceneFloatSection
      channels [None {type=MovieSceneScriptingFloatChannel, default=1, keys [[4936,0],[4961.1325,0.5],[5788.84,0.5],[5836.8675,1]]}]
     }]
   }
  ]
 }
 NiagaraComponent0 {
  id=#49
  parent_id=#22
  possessed_class=NiagaraComponent
  tracks [{
    type=MovieSceneParticleTrack
    display_name=FX System
    sections [{
      type=MovieSceneParticleSection
      channels [None {
        type=MovieSceneScriptingParticleChannel
        keys [
         [8035, "<ParticleKey.DEACTIVATE: 1>"]
         [8036, "<ParticleKey.ACTIVATE: 0>"]
         [8319, "<ParticleKey.ACTIVATE: 0>"]
         [8320, "<ParticleKey.DEACTIVATE: 1>"]
        ]
       }]
     }]
   }]
 }
 NS_PistolCanary_Spire_Beam_NtW {
  id=#8
  possessed_class=NiagaraActor
  tracks [{
    type=MovieSceneVisibilityTrack
    sections [{
      type=MovieSceneVisibilitySection
      channels [None {type=MovieSceneScriptingBoolChannel, default=false, keys [[5490.92,true],[5491.9325,false]]}]
     }]
   }]
 }
 NiagaraComponent0 {
  id=#50
  parent_id=#11
  possessed_class=NiagaraComponent
  tracks [
   {
    type=MovieSceneNiagaraVectorParameterTrack
    display_name=BeamEnd_0
    sections [{
      type=MovieSceneFloatVectorSection
      channels [
       Vector.X {type=MovieSceneScriptingFloatChannel, default=14348}
       Vector.Y {type=MovieSceneScriptingFloatChannel, default=69801}
       Vector.Z {type=MovieSceneScriptingFloatChannel, default=3921.7109}
      ]
     }]
   }
   {
    type=MovieSceneNiagaraFloatParameterTrack
    display_name=Duration
    sections [{type=MovieSceneFloatSection, channels [None {type=MovieSceneScriptingFloatChannel, default=3.6}]}]
   }
   {
    type=MovieSceneParticleTrack
    display_name=FX System
    sections [{
      type=MovieSceneParticleSection
      channels [None {
        type=MovieSceneScriptingParticleChannel
        keys [
         [6082, "<ParticleKey.DEACTIVATE: 1>"]
         [6083, "<ParticleKey.ACTIVATE: 0>"]
         [6304, "<ParticleKey.ACTIVATE: 0>"]
         [6305, "<ParticleKey.DEACTIVATE: 1>"]
        ]
       }]
     }]
   }
  ]
 }
 NiagaraComponent0 {
  id=#51
  parent_id=#23
  possessed_class=NiagaraComponent
  tracks [
   {
    type=MovieSceneNiagaraVectorParameterTrack
    display_name=BeamEnd_0
    sections [{
      type=MovieSceneFloatVectorSection
      channels [
       Vector.X {type=MovieSceneScriptingFloatChannel, default=26275}
       Vector.Y {type=MovieSceneScriptingFloatChannel, default=70300}
       Vector.Z {type=MovieSceneScriptingFloatChannel, default=6050}
      ]
     }]
   }
   {
    type=MovieSceneNiagaraFloatParameterTrack
    display_name=BeamFade
    sections [{type=MovieSceneFloatSection, channels [None {type=MovieSceneScriptingFloatChannel, default=1, keys [[5400,1],[5497,0]]}]}]
   }
  ]
 }
 NiagaraComponent0 {
  id=#52
  parent_id=#39
  possessed_class=NiagaraComponent
  tracks [
   {
    type=MovieSceneNiagaraVectorParameterTrack
    display_name=BeamEnd_0
    sections [{
      type=MovieSceneFloatVectorSection
      channels [
       Vector.X {type=MovieSceneScriptingFloatChannel, default=14284}
       Vector.Y {type=MovieSceneScriptingFloatChannel, default=57128}
       Vector.Z {type=MovieSceneScriptingFloatChannel, default=6100}
      ]
     }]
   }
   {
    type=MovieSceneNiagaraFloatParameterTrack
    display_name=BeamFade
    sections [{type=MovieSceneFloatSection, channels [None {type=MovieSceneScriptingFloatChannel, default=1, keys [[5398.67,1],[5497,0]]}]}]
   }
  ]
 }
 NiagaraComponent0 {
  id=#53
  parent_id=#42
  possessed_class=NiagaraComponent
  tracks [
   {type=MovieSceneNiagaraSystemTrack, sections [{type=MovieSceneNiagaraSystemSpawnSection, end_frame=9246, start_frame=6125}]}
   {
    type=MovieSceneNiagaraFloatParameterTrack
    display_name=SFXVolume
    sections [{type=MovieSceneFloatSection, channels [None {type=MovieSceneScriptingFloatChannel, default=0.3, keys [[6169,0],[6436,0.2],[6719,0.3]]}]}]
   }
   {
    type=MovieSceneParticleTrack
    display_name=FX System
    sections [{type=MovieSceneParticleSection, channels [None {type=MovieSceneScriptingParticleChannel, const ["<ParticleKey.TRIGGER: 2>", 1]}]}]
   }
  ]
 }
 NiagaraComponent0 {
  id=#54
  parent_id=#24
  possessed_class=NiagaraComponent
  tracks [
   {
    type=MovieSceneNiagaraFloatParameterTrack
    display_name=BeamFade
    sections [{type=MovieSceneFloatSection, channels [None {type=MovieSceneScriptingFloatChannel, default=1, keys [[5397,1],[5494,0]]}]}]
   }
   {
    type=MovieSceneParticleTrack
    display_name=FX System
    sections [{
      type=MovieSceneParticleSection
      channels [None {
        type=MovieSceneScriptingParticleChannel
        keys [
         [3541, "<ParticleKey.DEACTIVATE: 1>"]
         [3542, "<ParticleKey.ACTIVATE: 0>"]
         [5494, "<ParticleKey.ACTIVATE: 0>"]
         [5495, "<ParticleKey.DEACTIVATE: 1>"]
        ]
       }]
     }]
   }
  ]
 }
 NS_PC_MaskMaker_Smack_Shockwave {
  id=#44
  possessed_class=NiagaraActor
  tracks [{
    type=MovieScene3DTransformTrack
    sections [{
      type=MovieScene3DTransformSection
      channels [
       Location.X {type=MovieSceneScriptingDoubleChannel, default=14020}
       Location.Y {type=MovieSceneScriptingDoubleChannel, default=76216}
       Location.Z {type=MovieSceneScriptingDoubleChannel, default=7404}
       Rotation.X {type=MovieSceneScriptingDoubleChannel, default=0}
       Rotation.Y {type=MovieSceneScriptingDoubleChannel, default=0}
       Rotation.Z {type=MovieSceneScriptingDoubleChannel, default=0}
       Scale.X {type=MovieSceneScriptingDoubleChannel, default=1}
       Scale.Y {type=MovieSceneScriptingDoubleChannel, default=1}
       Scale.Z {type=MovieSceneScriptingDoubleChannel, default=1}
      ]
     }]
   }]
 }
 NiagaraComponent0 {
  id=#55
  parent_id=#29
  possessed_class=NiagaraComponent
  tracks [
   {
    type=MovieSceneNiagaraFloatParameterTrack
    display_name=BeamFade
    sections [{type=MovieSceneFloatSection, channels [None {type=MovieSceneScriptingFloatChannel, default=1, keys [[5397,1],[5494,0]]}]}]
   }
   {
    type=MovieSceneParticleTrack
    display_name=FX System
    sections [{
      type=MovieSceneParticleSection
      channels [None {
        type=MovieSceneScriptingParticleChannel
        keys [
         [3541, "<ParticleKey.DEACTIVATE: 1>"]
         [3542, "<ParticleKey.ACTIVATE: 0>"]
         [5494, "<ParticleKey.ACTIVATE: 0>"]
         [5495, "<ParticleKey.DEACTIVATE: 1>"]
        ]
       }]
     }]
   }
  ]
 }
]
display_rate {denominator=1, numerator=60}
tick_resolution {denominator=1, numerator=24000}
tracks [
 {
  type=MovieSceneMaterialParameterCollectionTrack
  display_name=MC_PC_Countdown_FX
  sections [{type=MovieSceneParameterSection, channels [Monolith_Barrier_Alpha {type=MovieSceneScriptingFloatChannel, default=1, const [1,1]}]}]
 }
 {type=MovieSceneSubTrack, display_name=Subsequences, sections [{type=MovieSceneSubSection, end_frame=8039, start_frame=5617}]}
]
