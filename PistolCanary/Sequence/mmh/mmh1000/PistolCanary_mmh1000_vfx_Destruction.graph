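# Level Sequence for the castle destruction VFX in the PistolCanary mmh1000 sequence.
# Niagara top-burst, castle crumble, tower crumble and ground-churn effects spawn in staggered order, with tower crumble placed near the SE tower.
# Four corner towers and four inner rock meshes spawn together, drop and rotate on keyed transforms rather than physics, then despawn.
# A material parameter collection track grows PortalRadius in PistolCanary_MPC as the collapse unfolds.
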LevelSequence /PistolCanary/Sequence/mmh/mmh1000/PistolCanary_mmh1000_vfx_Destruction
bindings [
 NS_PistolCanary_Castle_GroundChurn {
  id=#1
  possessed_class=NiagaraActor
  tracks [
   {
    type=MovieSceneSpawnTrack
    sections [{
      type=MovieSceneSpawnSection
      channels [None {type=MovieSceneScriptingBoolChannel, default=false, keys [[690,false],[810,true],[1375.54,false]]}]
     }]
   }
   {
    type=MovieScene3DTransformTrack
    sections [{
      type=MovieScene3DTransformSection
      channels [
       Location.X {type=MovieSceneScriptingDoubleChannel, default=14300}
       Location.Y {type=MovieSceneScriptingDoubleChannel, default=70200}
       Location.Z {type=MovieSceneScriptingDoubleChannel, default=3929}
       Rotation.X {type=MovieSceneScriptingDoubleChannel, default=0}
       Rotation.Y {type=MovieSceneScriptingDoubleChannel, default=0}
       Rotation.Z {type=MovieSceneScriptingDoubleChannel, default=0}
       Scale.X {type=MovieSceneScriptingDoubleChannel, default=1}
       Scale.Y {type=MovieSceneScriptingDoubleChannel, default=1}
       Scale.Z {type=MovieSceneScriptingDoubleChannel, default=1}
      ]
     }]
   }
  ]
 }
 SM_PistolCanary_CastleTower_NW {
  id=#2
  possessed_class=FortStaticMeshActor
  tracks [
   {
    type=MovieScene3DTransformTrack
    sections [{
      type=MovieScene3DTransformSection
      channels [
       Location.X {type=MovieSceneScriptingDoubleChannel, default=12544, keys [[810,12544],[1053.135,12752.6846]]}
       Location.Y {type=MovieSceneScriptingDoubleChannel, default=68052, const [68052,1]}
       Location.Z {type=MovieSceneScriptingDoubleChannel, default=3716, keys [[810,3716],[844.76,3640],[1053.135,1937.5443]]}
       Rotation.X {type=MovieSceneScriptingDoubleChannel, default=0, keys [[810,0],[1053.135,46.3411]]}
       Rotation.Y {type=MovieSceneScriptingDoubleChannel, default=0, keys [[810,0],[1053.135,-72.7551]]}
       Rotation.Z {type=MovieSceneScriptingDoubleChannel, default=0, keys [[810,0],[1053.135,-38.7675]]}
       Scale.X {type=MovieSceneScriptingDoubleChannel, default=1, const [1,1]}
       Scale.Y {type=MovieSceneScriptingDoubleChannel, default=1, const [1,1]}
       Scale.Z {type=MovieSceneScriptingDoubleChannel, default=1, const [1,1]}
      ]
     }]
   }
   {
    type=MovieSceneSpawnTrack
    sections [{
      type=MovieSceneSpawnSection
      channels [None {type=MovieSceneScriptingBoolChannel, default=false, keys [[250,false],[810,true],[1058.9525,false]]}]
     }]
   }
  ]
 }
 SM_PistolCanary_CastleTower_SW {
  id=#3
  possessed_class=FortStaticMeshActor
  tracks [
   {
    type=MovieScene3DTransformTrack
    sections [{
      type=MovieScene3DTransformSection
      channels [
       Location.X {type=MovieSceneScriptingDoubleChannel, default=12284, const [12284,1]}
       Location.Y {type=MovieSceneScriptingDoubleChannel, default=72320, const [72320,1]}
       Location.Z {type=MovieSceneScriptingDoubleChannel, default=3716, keys [[810,3716],[836.11,3640],[1049.805,1780.7638]]}
       Rotation.X {type=MovieSceneScriptingDoubleChannel, default=0, keys [[810,0],[1049.805,-32.4023]]}
       Rotation.Y {type=MovieSceneScriptingDoubleChannel, default=0, keys [[810,0],[1049.805,-26.5875]]}
       Rotation.Z {type=MovieSceneScriptingDoubleChannel, default=0, keys [[810,0],[1049.805,13.501]]}
       Scale.X {type=MovieSceneScriptingDoubleChannel, default=1, const [1,1]}
       Scale.Y {type=MovieSceneScriptingDoubleChannel, default=1, const [1,1]}
       Scale.Z {type=MovieSceneScriptingDoubleChannel, default=1, const [1,1]}
      ]
     }]
   }
   {
    type=MovieSceneSpawnTrack
    sections [{
      type=MovieSceneSpawnSection
      channels [None {type=MovieSceneScriptingBoolChannel, default=false, keys [[250,false],[810,true],[1058.9525,false]]}]
     }]
   }
  ]
 }
 InnerRocks4 {
  id=#4
  possessed_class=FortStaticMeshActor
  tracks [
   {
    type=MovieSceneSpawnTrack
    sections [{
      type=MovieSceneSpawnSection
      channels [None {type=MovieSceneScriptingBoolChannel, default=false, keys [[250,false],[810,true],[1004.875,false]]}]
     }]
   }
   {
    type=MovieScene3DTransformTrack
    sections [{
      type=MovieScene3DTransformSection
      channels [
       Location.X {type=MovieSceneScriptingDoubleChannel, default=13010.2516}
       Location.Y {type=MovieSceneScriptingDoubleChannel, default=70269.9275}
       Location.Z {type=MovieSceneScriptingDoubleChannel, default=1156, keys [[810,3596.8298],[1000.1025,1156]]}
       Rotation.X {type=MovieSceneScriptingDoubleChannel, default=0, keys [[810,0],[1000.1025,2.4814]]}
       Rotation.Y {type=MovieSceneScriptingDoubleChannel, default=0, keys [[810,0],[1000.1025,6.5254]]}
       Rotation.Z {type=MovieSceneScriptingDoubleChannel, default=180, keys [[810,180],[1000.1025,182.4178]]}
       Scale.X {type=MovieSceneScriptingDoubleChannel, default=1}
       Scale.Y {type=MovieSceneScriptingDoubleChannel, default=1}
       Scale.Z {type=MovieSceneScriptingDoubleChannel, default=1}
      ]
     }]
   }
  ]
 }
 InnerRocks1 {
  id=#5
  possessed_class=FortStaticMeshActor
  tracks [
   {
    type=MovieSceneSpawnTrack
    sections [{
      type=MovieSceneSpawnSection
      channels [None {type=MovieSceneScriptingBoolChannel, default=false, keys [[250,false],[810,true],[1004.875,false]]}]
     }]
   }
   {
    type=MovieScene3DTransformTrack
    sections [{
      type=MovieScene3DTransformSection
      channels [
       Location.X {type=MovieSceneScriptingDoubleChannel, default=14230.5398}
       Location.Y {type=MovieSceneScriptingDoubleChannel, default=71996.9381}
       Location.Z {type=MovieSceneScriptingDoubleChannel, default=1156, keys [[810,3656.7769],[982.595,1156]]}
       Rotation.X {type=MovieSceneScriptingDoubleChannel, default=0, keys [[810,0],[982.595,3.1455]]}
       Rotation.Y {type=MovieSceneScriptingDoubleChannel, default=0, keys [[810,0],[982.595,17.3976]]}
       Rotation.Z {type=MovieSceneScriptingDoubleChannel, default=90, keys [[810,90],[982.595,90.9414]]}
       Scale.X {type=MovieSceneScriptingDoubleChannel, default=1}
       Scale.Y {type=MovieSceneScriptingDoubleChannel, default=1}
       Scale.Z {type=MovieSceneScriptingDoubleChannel, default=1}
      ]
     }]
   }
  ]
 }
 NS_PistolCanary_Castle_TopBurst {
  id=#6
  possessed_class=NiagaraActor
  tracks [
   {
    type=MovieSceneSpawnTrack
    sections [{
      type=MovieSceneSpawnSection
      channels [None {type=MovieSceneScriptingBoolChannel, default=false, keys [[250,false],[491,true],[1375.54,false]]}]
     }]
   }
   {
    type=MovieScene3DTransformTrack
    sections [{
      type=MovieScene3DTransformSection
      channels [
       Location.X {type=MovieSceneScriptingDoubleChannel, default=14338.6198}
       Location.Y {type=MovieSceneScriptingDoubleChannel, default=70398.5749}
       Location.Z {type=MovieSceneScriptingDoubleChannel, default=7721.9627}
       Rotation.X {type=MovieSceneScriptingDoubleChannel, default=0}
       Rotation.Y {type=MovieSceneScriptingDoubleChannel, default=0}
       Rotation.Z {type=MovieSceneScriptingDoubleChannel, default=0}
       Scale.X {type=MovieSceneScriptingDoubleChannel, default=1}
       Scale.Y {type=MovieSceneScriptingDoubleChannel, default=1}
       Scale.Z {type=MovieSceneScriptingDoubleChannel, default=1}
      ]
     }]
   }
  ]
 }
 SM_PistolCanary_CastleTower_SE {
  id=#7
  possessed_class=FortStaticMeshActor
  tracks [
   {
    type=MovieScene3DTransformTrack
    sections [{
      type=MovieScene3DTransformSection
      channels [
       Location.X {type=MovieSceneScriptingDoubleChannel, default=16212, keys [[810,16212],[1043.75,16095.1338]]}
       Location.Y {type=MovieSceneScriptingDoubleChannel, default=72540, keys [[810,72540],[1043.75,72318.4062]]}
       Location.Z {type=MovieSceneScriptingDoubleChannel, default=3716, keys [[810,3716],[842.1075,3640],[1043.75,1373.4637]]}
       Rotation.X {type=MovieSceneScriptingDoubleChannel, default=0, keys [[810,0],[1043.75,-40.5649]]}
       Rotation.Y {type=MovieSceneScriptingDoubleChannel, default=0, keys [[810,0],[1043.75,32.6063]]}
       Rotation.Z {type=MovieSceneScriptingDoubleChannel, default=0, keys [[810,0],[1043.75,-9.4553]]}
       Scale.X {type=MovieSceneScriptingDoubleChannel, default=1, const [1,1]}
       Scale.Y {type=MovieSceneScriptingDoubleChannel, default=1, const [1,1]}
       Scale.Z {type=MovieSceneScriptingDoubleChannel, default=1, const [1,1]}
      ]
     }]
   }
   {
    type=MovieSceneSpawnTrack
    sections [{
      type=MovieSceneSpawnSection
      channels [None {type=MovieSceneScriptingBoolChannel, default=false, keys [[250,false],[810,true],[1058.9525,false]]}]
     }]
   }
  ]
 }
 SM_PistolCanary_CastleTower_NE {
  id=#8
  possessed_class=FortStaticMeshActor
  tracks [
   {
    type=MovieScene3DTransformTrack
    sections [{
      type=MovieScene3DTransformSection
      channels [
       Location.X {type=MovieSceneScriptingDoubleChannel, default=16332, const [16332,1]}
       Location.Y {type=MovieSceneScriptingDoubleChannel, default=68032, const [68032,1]}
       Location.Z {type=MovieSceneScriptingDoubleChannel, default=3716, keys [[810,3716],[840.24,3640],[1043.75,1468.2905]]}
       Rotation.X {type=MovieSceneScriptingDoubleChannel, default=0, keys [[810,0],[1043.75,84.3344]]}
       Rotation.Y {type=MovieSceneScriptingDoubleChannel, default=0, keys [[810,0],[1043.75,83.0374]]}
       Rotation.Z {type=MovieSceneScriptingDoubleChannel, default=0, keys [[810,0],[1043.75,26.1703]]}
       Scale.X {type=MovieSceneScriptingDoubleChannel, default=1, const [1,1]}
       Scale.Y {type=MovieSceneScriptingDoubleChannel, default=1, const [1,1]}
       Scale.Z {type=MovieSceneScriptingDoubleChannel, default=1, const [1,1]}
      ]
     }]
   }
   {
    type=MovieSceneSpawnTrack
    sections [{
      type=MovieSceneSpawnSection
      channels [None {type=MovieSceneScriptingBoolChannel, default=false, keys [[250,false],[810,true],[1058.9525,false]]}]
     }]
   }
  ]
 }
 NS_PistolCanary_Castle_Crumble {
  id=#9
  possessed_class=NiagaraActor
  tracks [
   {
    type=MovieSceneSpawnTrack
    sections [{
      type=MovieSceneSpawnSection
      channels [None {type=MovieSceneScriptingBoolChannel, default=false, keys [[250,false],[690,true],[1375.54,false]]}]
     }]
   }
   {
    type=MovieScene3DTransformTrack
    sections [{
      type=MovieScene3DTransformSection
      channels [
       Location.X {type=MovieSceneScriptingDoubleChannel, default=14338.6198}
       Location.Y {type=MovieSceneScriptingDoubleChannel, default=70398.5749}
       Location.Z {type=MovieSceneScriptingDoubleChannel, default=6923.8018}
       Rotation.X {type=MovieSceneScriptingDoubleChannel, default=0}
       Rotation.Y {type=MovieSceneScriptingDoubleChannel, default=0}
       Rotation.Z {type=MovieSceneScriptingDoubleChannel, default=0}
       Scale.X {type=MovieSceneScriptingDoubleChannel, default=1}
       Scale.Y {type=MovieSceneScriptingDoubleChannel, default=1}
       Scale.Z {type=MovieSceneScriptingDoubleChannel, default=1}
      ]
     }]
   }
  ]
 }
 InnerRocks2 {
  id=#10
  possessed_class=FortStaticMeshActor
  tracks [
   {
    type=MovieSceneSpawnTrack
    sections [{
      type=MovieSceneSpawnSection
      channels [None {type=MovieSceneScriptingBoolChannel, default=false, keys [[250,false],[810,true],[1004.875,false]]}]
     }]
   }
   {
    type=MovieScene3DTransformTrack
    sections [{
      type=MovieScene3DTransformSection
      channels [
       Location.X {type=MovieSceneScriptingDoubleChannel, default=15656.7793}
       Location.Y {type=MovieSceneScriptingDoubleChannel, default=70366.9679}
       Location.Z {type=MovieSceneScriptingDoubleChannel, default=1156, keys [[810,3605.9968],[996.4075,1156]]}
       Rotation.X {type=MovieSceneScriptingDoubleChannel, default=0, keys [[810,0],[996.4075,-3.8245]]}
       Rotation.Y {type=MovieSceneScriptingDoubleChannel, default=0, keys [[810,0],[996.4075,5.3147]]}
       Rotation.Z {type=MovieSceneScriptingDoubleChannel, default=0, keys [[810,0],[996.4075,-0.3548]]}
       Scale.X {type=MovieSceneScriptingDoubleChannel, default=1}
       Scale.Y {type=MovieSceneScriptingDoubleChannel, default=1}
       Scale.Z {type=MovieSceneScriptingDoubleChannel, default=1}
      ]
     }]
   }
  ]
 }
 InnerRocks3 {
  id=#11
  possessed_class=FortStaticMeshActor
  tracks [
   {
    type=MovieSceneSpawnTrack
    sections [{
      type=MovieSceneSpawnSection
      channels [None {type=MovieSceneScriptingBoolChannel, default=false, keys [[250,false],[810,true],[1004.875,false]]}]
     }]
   }
   {
    type=MovieScene3DTransformTrack
    sections [{
      type=MovieScene3DTransformSection
      channels [
       Location.X {type=MovieSceneScriptingDoubleChannel, default=14597.3741}
       Location.Y {type=MovieSceneScriptingDoubleChannel, default=68781.9251}
       Location.Z {type=MovieSceneScriptingDoubleChannel, default=1156, keys [[810,3569.1624],[988.875,1156]]}
       Rotation.X {type=MovieSceneScriptingDoubleChannel, default=0, keys [[810,0],[988.875,2.4805]]}
       Rotation.Y {type=MovieSceneScriptingDoubleChannel, default=0, keys [[810,0],[988.875,4.7261]]}
       Rotation.Z {type=MovieSceneScriptingDoubleChannel, default=-90, keys [[810,-90],[988.875,-91.9464]]}
       Scale.X {type=MovieSceneScriptingDoubleChannel, default=1}
       Scale.Y {type=MovieSceneScriptingDoubleChannel, default=1}
       Scale.Z {type=MovieSceneScriptingDoubleChannel, default=1}
      ]
     }]
   }
  ]
 }
 NS_PistolCanary_Tower_Crumble {
  id=#12
  possessed_class=NiagaraActor
  tracks [
   {
    type=MovieSceneSpawnTrack
    sections [{
      type=MovieSceneSpawnSection
      channels [None {type=MovieSceneScriptingBoolChannel, default=false, keys [[690,false],[805.91,true],[1375.54,false]]}]
     }]
   }
   {
    type=MovieScene3DTransformTrack
    sections [{
      type=MovieScene3DTransformSection
      channels [
       Location.X {type=MovieSceneScriptingDoubleChannel, default=16128.6757}
       Location.Y {type=MovieSceneScriptingDoubleChannel, default=72438.1612}
       Location.Z {type=MovieSceneScriptingDoubleChannel, default=6124.8953}
       Rotation.X {type=MovieSceneScriptingDoubleChannel, default=0}
       Rotation.Y {type=MovieSceneScriptingDoubleChannel, default=0}
       Rotation.Z {type=MovieSceneScriptingDoubleChannel, default=0}
       Scale.X {type=MovieSceneScriptingDoubleChannel, default=1}
       Scale.Y {type=MovieSceneScriptingDoubleChannel, default=1}
       Scale.Z {type=MovieSceneScriptingDoubleChannel, default=1}
      ]
     }]
   }
  ]
 }
]
display_rate {denominator=1, numerator=60}
tick_resolution {denominator=1, numerator=24000}
tracks [{
  type=MovieSceneMaterialParameterCollectionTrack
  display_name=PistolCanary_MPC
  sections [{
    type=MovieSceneParameterSection
    channels [PortalRadius {type=MovieSceneScriptingFloatChannel, default=4500, keys [[471,500],[955.71,4500]]}]
   }]
 }]
